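# Control Rig for the PistolCanary creature's cuttable tentacle with five eyes, with option channels toggling FK, spline, eye and cut-offset control visibility.
# Setup spawns eye controls by bone-name matching; eyes follow interpolated bone chains, with look channels driving offset curves.
# Procedural blink moves lids, orbs and balls by the larger of master or per-eye values.
# Cut Offset hands tentacle bones to offset controls, which can track FK until manipulated.
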ControlRigBlueprint /PistolCanary/Assets/Meshes/TentacleCut/TentacleCut_PistolCanary_Creature_CtrlRig
graphs [
 New Function_ContainedGraph {nodes [
   n1 = RigVMFunctionEntryNode()
   RigVMFunction_Sequence = &RigVMFunction_Sequence(ExecuteContext=@n1.ExecuteContext)
   RigVMFunctionReturnNode(ExecuteContext=@RigVMFunction_Sequence.B)
   n2 = &RigVMFunction_ControlFlowBranch(
    Condition=/Script/ControlRig.RigUnit_GetBoolAnimationChannel(Channel="Procedural Blink ", Control=control rig options, bInitial=False) {is=struct RigVMUnitNode, template="GetAnimationChannel::Execute(out Value,in Control,in Channel,in bInitial)"}.Value
    ExecuteContext=@RigVMFunction_Sequence.A
   )
   n3 = &RigVMFunction_Sequence(ExecuteContext=@n2.True)
   RigVMFunction_Sequence_6 = RigVMAggregateNode(ExecuteContext=@n3.A)
   n4 = /Script/ControlRig.RigUnit_GetFloatAnimationChannel(Channel=eyeA inner blink, Control=eyeA, bInitial=False) {is=struct RigVMUnitNode, template="GetAnimationChannel::Execute(out Value,in Control,in Channel,in bInitial)"}
   n5 = &RigUnit_GetFloatAnimationChannel(Channel=Master Inner Blink, Control=control rig options, bInitial=False) {template="GetAnimationChannel::Execute(out Value,in Control,in Channel,in bInitial)"}
   n6 = RigVMDispatchNode(
    True=@n4.Value
    Condition=/Script/RigVM.RigVMFunction_MathDoubleGreater(A=@n4.Value, B=@n5.Value) {is=struct RigVMUnitNode, template="Greater::Execute(in A,in B,out Result)"}.Result
    False=@n5.Value
   ) {template="DISPATCH_RigVMDispatch_If(in Condition,in True,in False,out Result)"}
   &RigUnit_ModifyTransforms(
    ItemToModify=((Item=(Type=Bone,Name="eyeA_innerlid_mstr"),Transform=(Rotation=(X=0,Y=-0.342,Z=0,W=0.9397),Translation=(X=0,Y=-20,Z=70),Scale3D=(X=1,Y=1,Z=1))))
    WeightMinimum=0
    Weight=&RigVMFunction_MathDoubleAbs(Value=&RigVMFunction_MathDoubleSub(B=1, A=@n6.Result).Result).Result
    ExecutePin=@RigVMFunction_Sequence_6.A
   )
   n7 = &RigUnit_GetFloatAnimationChannel(Channel=eyeB inner blink, Control=eyeB, bInitial=False) {template="GetAnimationChannel::Execute(out Value,in Control,in Channel,in bInitial)"}
   n8 = RigVMDispatchNode(
    True=@n7.Value
    Condition=&RigVMFunction_MathDoubleGreater(A=@n7.Value, B=@n5.Value) {template="Greater::Execute(in A,in B,out Result)"}.Result
    False=@n5.Value
   ) {template="DISPATCH_RigVMDispatch_If(in Condition,in True,in False,out Result)"}
   > &RigUnit_ModifyTransforms(
    ItemToModify=((Item=(Type=Bone,Name="eyeB_innerlid_mstr"),Transform=(Rotation=(X=0,Y=-0.2924,Z=0,W=0.9563),Translation=(X=0,Y=-10,Z=60),Scale3D=(X=1,Y=1,Z=1))))
    WeightMinimum=0
    Weight=&RigVMFunction_MathDoubleAbs(Value=&RigVMFunction_MathDoubleSub(B=1, A=@n8.Result).Result).Result
   )
   n9 = &RigUnit_GetFloatAnimationChannel(Channel=eyeC inner blink, Control=eyeC, bInitial=False) {template="GetAnimationChannel::Execute(out Value,in Control,in Channel,in bInitial)"}
   n10 = RigVMDispatchNode(
    True=@n9.Value
    Condition=&RigVMFunction_MathDoubleGreater(A=@n9.Value, B=@n5.Value) {template="Greater::Execute(in A,in B,out Result)"}.Result
    False=@n5.Value
   ) {template="DISPATCH_RigVMDispatch_If(in Condition,in True,in False,out Result)"}
   > &RigUnit_ModifyTransforms(
    ItemToModify=((Item=(Type=Bone,Name="eyeC_innerlid_mstr"),Transform=(Rotation=(X=0,Y=-0.2588,Z=0,W=0.9659),Translation=(X=-20,Y=0,Z=0),Scale3D=(X=1,Y=1,Z=1))))
    WeightMinimum=0
    Weight=&RigVMFunction_MathDoubleAbs(Value=&RigVMFunction_MathDoubleSub(B=1, A=@n10.Result).Result).Result
   )
   n11 = &RigUnit_GetFloatAnimationChannel(Channel=eyeD inner blink, Control=eyeD, bInitial=False) {template="GetAnimationChannel::Execute(out Value,in Control,in Channel,in bInitial)"}
   n12 = RigVMDispatchNode(
    True=@n11.Value
    Condition=&RigVMFunction_MathDoubleGreater(A=@n11.Value, B=@n5.Value) {template="Greater::Execute(in A,in B,out Result)"}.Result
    False=@n5.Value
   ) {template="DISPATCH_RigVMDispatch_If(in Condition,in True,in False,out Result)"}
   > &RigUnit_ModifyTransforms(
    ItemToModify=((Item=(Type=Bone,Name="eyeD_innerlid_mstr"),Transform=(Rotation=(X=-0.0113,Y=-0.2586,Z=-0.0421,W=0.965),Translation=(X=-10,Y=0,Z=0),Scale3D=(X=1,Y=1,Z=1))))
    WeightMinimum=0
    Weight=&RigVMFunction_MathDoubleAbs(Value=&RigVMFunction_MathDoubleSub(B=1, A=@n12.Result).Result).Result
   )
   n13 = &RigUnit_GetFloatAnimationChannel(Channel=eyeE inner blink, Control=eyeE, bInitial=False) {template="GetAnimationChannel::Execute(out Value,in Control,in Channel,in bInitial)"}
   n14 = RigVMDispatchNode(
    True=@n13.Value
    Condition=&RigVMFunction_MathDoubleGreater(A=@n13.Value, B=@n5.Value) {template="Greater::Execute(in A,in B,out Result)"}.Result
    False=@n5.Value
   ) {template="DISPATCH_RigVMDispatch_If(in Condition,in True,in False,out Result)"}
   > &RigUnit_ModifyTransforms(
    ItemToModify=((Item=(Type=Bone,Name="eyeE_innerlid_mstr"),Transform=(Rotation=(X=0,Y=-0.3007,Z=0,W=0.9537),Translation=(X=-20,Y=0,Z=0),Scale3D=(X=1,Y=1,Z=1))))
    WeightMinimum=0
    Weight=&RigVMFunction_MathDoubleAbs(Value=&RigVMFunction_MathDoubleSub(B=1, A=@n14.Result).Result).Result
   )
   &RigUnit_ModifyTransforms(
    ItemToModify=((Item=(Type=Bone,Name="eyeA_orb"),Transform=(Rotation=(X=0,Y=0,Z=0,W=1),Translation=(X=-30,Y=0,Z=0),Scale3D=(X=1,Y=1,Z=1))))
    WeightMinimum=0
    ExecutePin=@RigVMFunction_Sequence_6.C
    Weight=@n6.Result
   )
   > &RigUnit_ModifyTransforms(
    ItemToModify=((Item=(Type=Bone,Name="eyeA_ball"),Transform=(Rotation=(X=0,Y=0,Z=0,W=1),Translation=(X=-150,Y=0,Z=0),Scale3D=(X=1,Y=1,Z=1))))
    WeightMinimum=0
    Weight=@n6.Result
   )
   > &RigUnit_ModifyTransforms(
    ItemToModify=((Item=(Type=Bone,Name="eyeB_orb"),Transform=(Rotation=(X=0,Y=0,Z=0,W=1),Translation=(X=-50,Y=-20,Z=0),Scale3D=(X=1,Y=1,Z=1))))
    WeightMinimum=0
    Weight=@n8.Result
   )
   > &RigUnit_ModifyTransforms(
    ItemToModify=((Item=(Type=Bone,Name="eyeB_ball"),Transform=(Rotation=(X=0,Y=0,Z=0,W=1),Translation=(X=-50,Y=0,Z=0),Scale3D=(X=1,Y=1,Z=1))))
    WeightMinimum=0
    Weight=@n8.Result
   )
   > &RigUnit_ModifyTransforms(
    ItemToModify=((Item=(Type=Bone,Name="eyeC_orb"),Transform=(Rotation=(X=0,Y=0,Z=0,W=1),Translation=(X=-40,Y=0,Z=0),Scale3D=(X=1,Y=1,Z=1))))
    WeightMinimum=0
    Weight=@n10.Result
   )
   > &RigUnit_ModifyTransforms(
    ItemToModify=((Item=(Type=Bone,Name="eyeC_ball"),Transform=(Rotation=(X=0,Y=0,Z=0,W=1),Translation=(X=-30,Y=0,Z=0),Scale3D=(X=1,Y=1,Z=1))))
    WeightMinimum=0
    Weight=@n10.Result
   )
   > &RigUnit_ModifyTransforms(
    ItemToModify=((Item=(Type=Bone,Name="eyeD_orb"),Transform=(Rotation=(X=0,Y=0,Z=0,W=1),Translation=(X=-30,Y=0,Z=0),Scale3D=(X=1,Y=1,Z=1))))
    WeightMinimum=0
    Weight=@n12.Result
   )
   > &RigUnit_ModifyTransforms(
    ItemToModify=((Item=(Type=Bone,Name="eyeD_ball"),Transform=(Rotation=(X=0,Y=0,Z=0,W=1),Translation=(X=-30,Y=0,Z=0),Scale3D=(X=1,Y=1,Z=1))))
    WeightMinimum=0
    Weight=@n12.Result
   )
   > &RigUnit_ModifyTransforms(
    ItemToModify=((Item=(Type=Bone,Name="eyeE_orb"),Transform=(Rotation=(X=0,Y=0,Z=0,W=1),Translation=(X=-20,Y=0,Z=0),Scale3D=(X=1,Y=1,Z=1))))
    WeightMinimum=0
    Weight=@n14.Result
   )
   > &RigUnit_ModifyTransforms(
    ItemToModify=((Item=(Type=Bone,Name="eyeE_ball"),Transform=(Rotation=(X=0,Y=0,Z=0,W=1),Translation=(X=-30,Y=0,Z=0),Scale3D=(X=1,Y=1,Z=1))))
    WeightMinimum=0
    Weight=@n14.Result
   )
   &RigUnit_ModifyTransforms(
    ItemToModify=((Item=(Type=Bone,Name="eyeA_innerlid_mstr"),Transform=(Rotation=(X=0,Y=0.3827,Z=0,W=0.9239),Translation=(X=5,Y=0,Z=0),Scale3D=(X=1,Y=1,Z=1))))
    WeightMinimum=0
    ExecutePin=@RigVMFunction_Sequence_6.B
    Weight=@n6.Result
   )
   > &RigUnit_ModifyTransforms(
    ItemToModify=((Item=(Type=Bone,Name="eyeB_innerlid_mstr"),Transform=(Rotation=(X=-0.02,Y=0.3822,Z=0.0484,W=0.9226),Translation=(X=30,Y=0,Z=0),Scale3D=(X=1,Y=1,Z=1))))
    WeightMinimum=0
    Weight=@n8.Result
   )
   > &RigUnit_ModifyTransforms(
    ItemToModify=((Item=(Type=Bone,Name="eyeC_innerlid_mstr"),Transform=(Rotation=(X=0,Y=0.3007,Z=0,W=0.9537),Translation=(X=0,Y=0,Z=0),Scale3D=(X=1,Y=1,Z=1))))
    WeightMinimum=0
    Weight=@n10.Result
   )
   > &RigUnit_ModifyTransforms(
    ItemToModify=((Item=(Type=Bone,Name="eyeD_innerlid_mstr"),Transform=(Rotation=(X=0.0033,Y=0.3007,Z=-0.0104,W=0.9537),Translation=(X=10,Y=0,Z=5),Scale3D=(X=1,Y=1,Z=1))))
    WeightMinimum=0
    Weight=@n12.Result
   )
   > &RigUnit_ModifyTransforms(
    ItemToModify=((Item=(Type=Bone,Name="eyeE_innerlid_mstr"),Transform=(Rotation=(X=0,Y=0.342,Z=0,W=0.9397),Translation=(X=0,Y=0,Z=0),Scale3D=(X=1,Y=1,Z=1))))
    WeightMinimum=0
    Weight=@n14.Result
   )
   RigVMFunction_Sequence_7 = RigVMAggregateNode(ExecuteContext=@n3.B)
   n15 = &RigUnit_GetFloatAnimationChannel(Channel=eyeA blink, Control=eyeA, bInitial=False) {template="GetAnimationChannel::Execute(out Value,in Control,in Channel,in bInitial)"}
   n16 = &RigUnit_GetFloatAnimationChannel(Channel=Master Blink, Control=control rig options, bInitial=False) {template="GetAnimationChannel::Execute(out Value,in Control,in Channel,in bInitial)"}
   n17 = RigVMDispatchNode(
    True=@n15.Value
    Condition=&RigVMFunction_MathDoubleGreater(A=@n15.Value, B=@n16.Value) {template="Greater::Execute(in A,in B,out Result)"}.Result
    False=@n16.Value
   ) {template="DISPATCH_RigVMDispatch_If(in Condition,in True,in False,out Result)"}
   &RigUnit_ModifyTransforms(
    ItemToModify=((Item=(Type=Bone,Name="eyeA_lid_upper_master"),Transform=(Rotation=(X=0,Y=0,Z=-0.3827,W=0.9239),Translation=(X=0,Y=0,Z=0),Scale3D=(X=1,Y=1,Z=1))))
    WeightMinimum=0
    Weight=@n17.Result
    ExecutePin=@RigVMFunction_Sequence_7.A
   )
   n18 = &RigUnit_GetFloatAnimationChannel(Channel=eyeB blink, Control=eyeB, bInitial=False) {template="GetAnimationChannel::Execute(out Value,in Control,in Channel,in bInitial)"}
   n19 = RigVMDispatchNode(
    True=@n18.Value
    Condition=&RigVMFunction_MathDoubleGreater(A=@n18.Value, B=@n16.Value) {template="Greater::Execute(in A,in B,out Result)"}.Result
    False=@n16.Value
   ) {template="DISPATCH_RigVMDispatch_If(in Condition,in True,in False,out Result)"}
   > &RigUnit_ModifyTransforms(
    ItemToModify=((Item=(Type=Bone,Name="eyeB_lid_upper_master"),Transform=(Rotation=(X=0,Y=0,Z=-0.3827,W=0.9239),Translation=(X=0,Y=0,Z=0),Scale3D=(X=1,Y=1,Z=1))))
    WeightMinimum=0
    Weight=@n19.Result
   )
   n20 = &RigUnit_GetFloatAnimationChannel(Channel=eyeC blink, Control=eyeC, bInitial=False) {template="GetAnimationChannel::Execute(out Value,in Control,in Channel,in bInitial)"}
   n21 = RigVMDispatchNode(
    True=@n20.Value
    Condition=&RigVMFunction_MathDoubleGreater(A=@n20.Value, B=@n16.Value) {template="Greater::Execute(in A,in B,out Result)"}.Result
    False=@n16.Value
   ) {template="DISPATCH_RigVMDispatch_If(in Condition,in True,in False,out Result)"}
   > &RigUnit_ModifyTransforms(
    ItemToModify=((Item=(Type=Bone,Name="eyeC_lid_upper_master"),Transform=(Rotation=(X=0,Y=0,Z=-0.1736,W=0.9848),Translation=(X=0,Y=0,Z=0),Scale3D=(X=1,Y=1,Z=1))))
    WeightMinimum=0
    Weight=@n21.Result
   )
   n22 = &RigUnit_GetFloatAnimationChannel(Channel=eyeD blink, Control=eyeD, bInitial=False) {template="GetAnimationChannel::Execute(out Value,in Control,in Channel,in bInitial)"}
   n23 = RigVMDispatchNode(
    True=@n22.Value
    Condition=&RigVMFunction_MathDoubleGreater(A=@n22.Value, B=@n16.Value) {template="Greater::Execute(in A,in B,out Result)"}.Result
    False=@n16.Value
   ) {template="DISPATCH_RigVMDispatch_If(in Condition,in True,in False,out Result)"}
   > &RigUnit_ModifyTransforms(
    ItemToModify=((Item=(Type=Bone,Name="eyeD_lid_upper_master"),Transform=(Rotation=(X=0,Y=0,Z=-0.2588,W=0.9659),Translation=(X=0,Y=0,Z=0),Scale3D=(X=1,Y=1,Z=1))))
    WeightMinimum=0
    Weight=@n23.Result
   )
   n24 = &RigUnit_GetFloatAnimationChannel(Channel=eyeE blink, Control=eyeE, bInitial=False) {template="GetAnimationChannel::Execute(out Value,in Control,in Channel,in bInitial)"}
   n25 = RigVMDispatchNode(
    True=@n24.Value
    Condition=&RigVMFunction_MathDoubleGreater(A=@n24.Value, B=@n16.Value) {template="Greater::Execute(in A,in B,out Result)"}.Result
    False=@n16.Value
   ) {template="DISPATCH_RigVMDispatch_If(in Condition,in True,in False,out Result)"}
   > &RigUnit_ModifyTransforms(
    ItemToModify=((Item=(Type=Bone,Name="eyeE_lid_upper_master"),Transform=(Rotation=(X=0,Y=0,Z=-0.1305,W=0.9914),Translation=(X=0,Y=0,Z=0),Scale3D=(X=1,Y=1,Z=1))))
    WeightMinimum=0
    Weight=@n25.Result
   )
   &RigUnit_ModifyTransforms(
    ItemToModify=((Item=(Type=Bone,Name="eyeA_lid_lower_master"),Transform=(Rotation=(X=0,Y=0,Z=-0.1305,W=0.9914),Translation=(X=0,Y=0,Z=0),Scale3D=(X=1,Y=1,Z=1))))
    WeightMinimum=0
    ExecutePin=@RigVMFunction_Sequence_7.B
    Weight=@n17.Result
   )
   > &RigUnit_ModifyTransforms(
    ItemToModify=((Item=(Type=Bone,Name="eyeB_lid_lower_master"),Transform=(Rotation=(X=0,Y=0,Z=-0.1736,W=0.9848),Translation=(X=0,Y=0,Z=0),Scale3D=(X=1,Y=1,Z=1))))
    WeightMinimum=0
    Weight=@n19.Result
   )
   > &RigUnit_ModifyTransforms(
    ItemToModify=((Item=(Type=Bone,Name="eyeC_lid_lower_master"),Transform=(Rotation=(X=0,Y=0,Z=-0.1736,W=0.9848),Translation=(X=0,Y=0,Z=0),Scale3D=(X=1,Y=1,Z=1))))
    WeightMinimum=0
    Weight=@n21.Result
   )
   > &RigUnit_ModifyTransforms(
    ItemToModify=((Item=(Type=Bone,Name="eyeD_lid_lower_master"),Transform=(Rotation=(X=0,Y=0,Z=-0.1305,W=0.9914),Translation=(X=0,Y=0,Z=0),Scale3D=(X=1,Y=1,Z=1))))
    WeightMinimum=0
    Weight=@n23.Result
   )
   > &RigUnit_ModifyTransforms(
    ItemToModify=((Item=(Type=Bone,Name="eyeE_lid_lower_master"),Transform=(Rotation=(X=0,Y=0,Z=-0.1305,W=0.9914),Translation=(X=0,Y=0,Z=0),Scale3D=(X=1,Y=1,Z=1))))
    WeightMinimum=0
    Weight=@n25.Result
   )
   &RigUnit_ModifyTransforms(
    ItemToModify=((Item=(Type=Bone,Name="eyeA_orb"),Transform=(Rotation=(X=0,Y=0,Z=0,W=1),Translation=(X=-100,Y=0,Z=0),Scale3D=(X=1,Y=1,Z=1))))
    WeightMinimum=0
    ExecutePin=@RigVMFunction_Sequence_7.C
    Weight=@n17.Result
   )
   > &RigUnit_ModifyTransforms(
    ItemToModify=((Item=(Type=Bone,Name="eyeB_orb"),Transform=(Rotation=(X=0,Y=0,Z=0,W=1),Translation=(X=-100,Y=0,Z=0),Scale3D=(X=1,Y=1,Z=1))))
    WeightMinimum=0
    Weight=@n19.Result
   )
   > &RigUnit_ModifyTransforms(
    ItemToModify=((Item=(Type=Bone,Name="eyeC_orb"),Transform=(Rotation=(X=0,Y=0,Z=0,W=1),Translation=(X=-100,Y=0,Z=0),Scale3D=(X=1,Y=1,Z=1))))
    WeightMinimum=0
    Weight=@n21.Result
   )
   > &RigUnit_ModifyTransforms(
    ItemToModify=((Item=(Type=Bone,Name="eyeD_orb"),Transform=(Rotation=(X=0,Y=0,Z=0,W=1),Translation=(X=-100,Y=0,Z=0),Scale3D=(X=1,Y=1,Z=1))))
    WeightMinimum=0
    Weight=@n23.Result
   )
   > &RigUnit_ModifyTransforms(
    ItemToModify=((Item=(Type=Bone,Name="eyeE_orb"),Transform=(Rotation=(X=0,Y=0,Z=0,W=1),Translation=(X=-60,Y=0,Z=0),Scale3D=(X=1,Y=1,Z=1))))
    WeightMinimum=0
    Weight=@n25.Result
   )
  ]}
 New Function_ContainedGraph {nodes [
   n1 = RigVMFunctionEntryNode()
   RigVMFunction_Sequence = &RigVMFunction_Sequence(ExecuteContext=@n1.ExecuteContext)
   RigVMFunctionReturnNode(ExecuteContext=@RigVMFunction_Sequence.B)
   RigVMFunction_Sequence_1 = RigVMAggregateNode(ExecuteContext=@RigVMFunction_Sequence.A)
   RigVMFunction_Sequence_5 = RigVMAggregateNode(ExecuteContext=@RigVMFunction_Sequence_1.D)
   For_Each_5 = RigVMDispatchNode(Array=RigVMVariableNode() {variable=eye control array}.Value, ExecuteContext=@RigVMFunction_Sequence_5.B) {template="DISPATCH_RigVMDispatch_ArrayIterator(in Array,out Element,out Index,out Count,out Ratio)"}
   > /Script/ControlRig.RigUnit_SetControlVisibility(
    Item=@For_Each_5.Element
    bVisible=&RigUnit_GetBoolAnimationChannel(Channel=Eye Control Visibility, Control=control rig options, bInitial=false) {template="GetAnimationChannel::Execute(out Value,in Control,in Channel,in bInitial)"}.Value
   ) {is=struct RigVMUnitNode, template="SetControlVisibility::Execute(in Item,in Pattern,in bVisible)"}
   n2 = &RigVMFunction_Sequence(ExecuteContext=@RigVMFunction_Sequence_5.C)
   For_Each_12 = RigVMDispatchNode(ExecuteContext=@n2.A, Array=RigVMVariableNode() {variable=innerlid control array}.Value) {template="DISPATCH_RigVMDispatch_ArrayIterator(in Array,out Element,out Index,out Count,out Ratio)"}
   RigVMFunction_MathBoolNot = &RigVMFunction_MathBoolNot(Value=&RigUnit_GetBoolAnimationChannel(Channel="Procedural Blink ", Control=control rig options, bInitial=false) {template="GetAnimationChannel::Execute(out Value,in Control,in Channel,in bInitial)"}.Value)
   > &RigUnit_SetControlVisibility(Item=@For_Each_12.Element, bVisible=@RigVMFunction_MathBoolNot.Result) {template="SetControlVisibility::Execute(in Item,in Pattern,in bVisible)"}
   > &RigUnit_SetControlVisibility(
    Item=RigVMDispatchNode(Array=RigVMVariableNode() {variable=eye ball control array}.Value, Index=@For_Each_12.Index) {template="DISPATCH_RigVMDispatch_ArrayGetAtIndex(in Array,in Index,out Element)"}.Element
    bVisible=@RigVMFunction_MathBoolNot.Result
   ) {template="SetControlVisibility::Execute(in Item,in Pattern,in bVisible)"}
   For_Each_7 = RigVMDispatchNode(Array=RigVMVariableNode() {variable=lid control array}.Value, ExecuteContext=@n2.B) {template="DISPATCH_RigVMDispatch_ArrayIterator(in Array,out Element,out Index,out Count,out Ratio)"}
   > &RigUnit_SetControlVisibility(Item=@For_Each_7.Element, bVisible=@RigVMFunction_MathBoolNot.Result) {template="SetControlVisibility::Execute(in Item,in Pattern,in bVisible)"}
   n3 = &RigVMFunction_Sequence(ExecuteContext=@RigVMFunction_Sequence_5.A)
   For_Each_11 = RigVMDispatchNode(Array=RigVMVariableNode() {variable=offset tentacle controls}.Value, ExecuteContext=@n3.A) {template="DISPATCH_RigVMDispatch_ArrayIterator(in Array,out Element,out Index,out Count,out Ratio)"}
   n4 = &RigUnit_GetBoolAnimationChannel(Channel=Enable Cut Offset, Control=control rig options, bInitial=false) {template="GetAnimationChannel::Execute(out Value,in Control,in Channel,in bInitial)"}
   > &RigUnit_SetControlVisibility(Item=@For_Each_11.Element, bVisible=@n4.Value) {template="SetControlVisibility::Execute(in Item,in Pattern,in bVisible)"}
   &RigUnit_SetControlVisibility(Item=(Type=Control,Name="tentacle_08_dup_offset"), ExecutePin=@For_Each_11.Completed, bVisible=@n4.Value) {template="SetControlVisibility::Execute(in Item,in Pattern,in bVisible)"}
   For_Each_16 = RigVMDispatchNode(Array=RigVMVariableNode() {variable=upper tentacle controls}.Value, ExecuteContext=@n3.B) {template="DISPATCH_RigVMDispatch_ArrayIterator(in Array,out Element,out Index,out Count,out Ratio)"}
   > &RigUnit_SetControlVisibility(Item=@For_Each_16.Element, bVisible=&RigVMFunction_MathBoolNot(Value=@n4.Value).Result) {template="SetControlVisibility::Execute(in Item,in Pattern,in bVisible)"}
   &RigUnit_SetControlVisibility(
    bVisible=&RigUnit_GetBoolAnimationChannel(Channel=Root Visibility, Control=control rig options, bInitial=false) {template="GetAnimationChannel::Execute(out Value,in Control,in Channel,in bInitial)"}.Value
    ExecutePin=@RigVMFunction_Sequence_1.A
    Item=RigVMDispatchNode(Index=0, Array=RigVMVariableNode() {variable=control fk array}.Value) {template="DISPATCH_RigVMDispatch_ArrayGetAtIndex(in Array,in Index,out Element)"}.Element
   ) {template="SetControlVisibility::Execute(in Item,in Pattern,in bVisible)"}
   For_Each_3 = RigVMDispatchNode(Array=RigVMVariableNode() {variable=animated ctrl array}.Value, ExecuteContext=@RigVMFunction_Sequence_1.B) {template="DISPATCH_RigVMDispatch_ArrayIterator(in Array,out Element,out Index,out Count,out Ratio)"}
   > SetControlVisibility = &RigUnit_SetControlVisibility(
    Item=@For_Each_3.Element
    bVisible=&RigUnit_GetBoolAnimationChannel(Channel=FK Rig Visibility, Control=control rig options, bInitial=False) {template="GetAnimationChannel::Execute(out Value,in Control,in Channel,in bInitial)"}.Value
   ) {template="SetControlVisibility::Execute(in Item,in Pattern,in bVisible)"}
   n5 = &RigVMFunction_Sequence(ExecuteContext=@RigVMFunction_Sequence_1.C)
   For_Each_4 = RigVMDispatchNode(ExecuteContext=@n5.A, Array=RigVMVariableNode() {variable=spline direct ctrl array}.Value) {template="DISPATCH_RigVMDispatch_ArrayIterator(in Array,out Element,out Index,out Count,out Ratio)"}
   n6 = &RigUnit_GetBoolAnimationChannel(Channel=Spline Rig Visibility, Control=control rig options, bInitial=False) {template="GetAnimationChannel::Execute(out Value,in Control,in Channel,in bInitial)"}
   > &RigUnit_SetControlVisibility(Item=@For_Each_4.Element, bVisible=@n6.Value) {template="SetControlVisibility::Execute(in Item,in Pattern,in bVisible)"}
   For_Each_1 = RigVMDispatchNode(Array=RigVMVariableNode() {variable=spline ctrl array}.Value, ExecuteContext=@n5.B) {template="DISPATCH_RigVMDispatch_ArrayIterator(in Array,out Element,out Index,out Count,out Ratio)"}
   > &RigUnit_SetControlVisibility(Item=@For_Each_1.Element, bVisible=@n6.Value) {template="SetControlVisibility::Execute(in Item,in Pattern,in bVisible)"}
  ]}
 New Function_ContainedGraph {nodes [
   RigVMFunctionEntryNode()
   > RigVMFunction_Sequence_9 = RigVMAggregateNode()
   RigVMFunctionReturnNode(ExecuteContext=@RigVMFunction_Sequence_9.E)
   For_Each_10 = RigVMDispatchNode(
    Array=((Type=Bone,Name="eyeE"),(Type=Bone,Name="eyeD"),(Type=Bone,Name="eyeC"),(Type=Bone,Name="eyeB"),(Type=Bone,Name="eyeA"))
    ExecuteContext=@RigVMFunction_Sequence_9.A
   ) {template="DISPATCH_RigVMDispatch_ArrayIterator(in Array,out Element,out Index,out Count,out Ratio)"}
   n1 = /Script/ControlRig.RigUnit_ChainInfo(
    DebugScale=1
    Items=((Type=None,Name="None"),(Type=None,Name="None"),(Type=None,Name="None"))
    bCalculateStretch=True
    bDebug=False
    bInitial=False
    Items.2=RigVMDispatchNode(Array=RigVMVariableNode() {variable=eye interp end bones}.Value, Index=@For_Each_10.Index) {template="DISPATCH_RigVMDispatch_ArrayGetAtIndex(in Array,in Index,out Element)"}.Element
    Param=RigVMDispatchNode(Array=RigVMVariableNode() {variable=eye interp param}.Value, Index=@For_Each_10.Index) {template="DISPATCH_RigVMDispatch_ArrayGetAtIndex(in Array,in Index,out Element)"}.Element
    Items.0=RigVMDispatchNode(Array=RigVMVariableNode() {variable=eye interp start bones}.Value, Index=@For_Each_10.Index) {template="DISPATCH_RigVMDispatch_ArrayGetAtIndex(in Array,in Index,out Element)"}.Element
    Items.1=RigVMDispatchNode(Array=RigVMVariableNode() {variable=eye interp mid bones}.Value, Index=@For_Each_10.Index) {template="DISPATCH_RigVMDispatch_ArrayGetAtIndex(in Array,in Index,out Element)"}.Element
   ) {is=struct RigVMUnitNode}
   > &RigUnit_SetTransform(
    Value=(Rotation=(X=0,Y=0,Z=0,W=1),Translation=(X=0,Y=0,Z=0),Scale3D=(X=1,Y=1,Z=1))
    Weight=1
    bInitial=False
    bPropagateToChildren=true
    Value.Rotation=@n1.InterpolatedTransform.Rotation
    Value.Scale3D.Z=@n1.InterpolatedTransform.Scale3D.X
    Value.Scale3D.Y=@n1.InterpolatedTransform.Scale3D.Y
    Value.Scale3D.X=@n1.InterpolatedTransform.Scale3D.Z
    Value.Translation=@n1.InterpolatedTransform.Translation
    Item=@For_Each_10.Element
   )
   For_Each_6 = RigVMDispatchNode(Array=RigVMVariableNode() {variable=eye matching skel array}.Value, ExecuteContext=@RigVMFunction_Sequence_9.C) {template="DISPATCH_RigVMDispatch_ArrayIterator(in Array,out Element,out Index,out Count,out Ratio)"}
   > Name Check = RigVMFunctionReferenceNode(Names to Check=("innerlid"), Element=@For_Each_6.Element)
   n2 = RigVMDispatchNode(Index=@For_Each_6.Index, Array=RigVMVariableNode() {variable=eye control array}.Value) {template="DISPATCH_RigVMDispatch_ArrayGetAtIndex(in Array,in Index,out Element)"}
   RigUnit_GetTransform = &RigUnit_GetTransform(bInitial=False, Item=@n2.Element)
   > &RigUnit_SetTransform(
    Value=(Rotation=(X=0,Y=0,Z=0,W=1),Translation=(X=0,Y=0,Z=0),Scale3D=(X=1,Y=1,Z=1))
    Weight=1
    bInitial=false
    bPropagateToChildren=true
    Value.Scale3D=RigVMDispatchNode(
     True=(X=0,Y=0,Z=0)
     Condition=@"Name Check".Result
     False=@RigUnit_GetTransform.Transform.Scale3D
     True.Z=@RigUnit_GetTransform.Transform.Scale3D.X
     True.Y=@RigUnit_GetTransform.Transform.Scale3D.Y
     True.X=@RigUnit_GetTransform.Transform.Scale3D.Z
    ) {template="DISPATCH_RigVMDispatch_If(in Condition,in True,in False,out Result)"}.Result
    Value.Rotation=@RigUnit_GetTransform.Transform.Rotation
    Value.Translation=@RigUnit_GetTransform.Transform.Translation
    Item=@For_Each_6.Element
   )
   /Script/ControlRig.RigUnit_SetCurveValue(
    Curve=Eye_Y_Offset
    Value=&RigUnit_GetFloatAnimationChannel(Channel=Eyes Look Y, Control=control rig options, bInitial=False) {template="GetAnimationChannel::Execute(out Value,in Control,in Channel,in bInitial)"}.Value
    ExecutePin=@RigVMFunction_Sequence_9.D
   ) {is=struct RigVMUnitNode}
   > &RigUnit_SetCurveValue(
    Curve=Eye_X_Offset
    Value=&RigUnit_GetFloatAnimationChannel(Channel=Eyes Look X, Control=control rig options, bInitial=False) {template="GetAnimationChannel::Execute(out Value,in Control,in Channel,in bInitial)"}.Value
   )
   For_Each = RigVMDispatchNode(
    Array=((Type=Control,Name="eyeE"),(Type=Control,Name="eyeD"),(Type=Control,Name="eyeC"),(Type=Control,Name="eyeB"),(Type=Control,Name="eyeA"))
    ExecuteContext=@RigVMFunction_Sequence_9.B
   ) {template="DISPATCH_RigVMDispatch_ArrayIterator(in Array,out Element,out Index,out Count,out Ratio)"}
   Make_Array = RigVMDispatchNode(Values=((Type=Bone,Name="eyeE"),(Type=Bone,Name="eyeD"),((),Type=Bone,Name="eyeC"),((),Type=Bone,Name="eyeB"),((),Type=Bone,Name="eyeA"))) {template="DISPATCH_RigVMDispatch_ArrayMake(in Values,out Array)"}
   n3 = RigVMDispatchNode(Array=@Make_Array.Array, Index=@For_Each.Index) {template="DISPATCH_RigVMDispatch_ArrayGetAtIndex(in Array,in Index,out Element)"}
   n4 = RigVMDispatchNode(Array=RigVMVariableNode() {variable=eye null array}.Value, Index=@For_Each.Index) {template="DISPATCH_RigVMDispatch_ArrayGetAtIndex(in Array,in Index,out Element)"}
   Multiply = &RigVMFunction_MathTransformMul(
    B=&RigUnit_GetTransform(bInitial=False, Item=@n3.Element).Transform
    A=&RigVMFunction_MathTransformMakeRelative(
     Global=&RigUnit_GetTransform(bInitial=true, Item=@n3.Element).Transform
     Parent=&RigUnit_GetTransform(bInitial=true, Item=@n4.Element).Transform
    ).Local
   ) {template="Multiply::Execute(in A,in B,out Result)"}
   > SetControlOffset = /Script/ControlRig.RigUnit_SetControlOffset(Space=GlobalSpace, Offset=@Multiply.Result, Control=@For_Each.Element.Name) {is=struct RigVMUnitNode, template="SetControlOffset::Execute(in Control,io Offset,in Space)"}
   For_Each_1 = RigVMDispatchNode(ExecuteContext=(), Array=RigVMVariableNode() {variable=eye leftovers joints}.Value) {template="DISPATCH_RigVMDispatch_ArrayIterator(in Array,out Element,out Index,out Count,out Ratio)"}
   n5 = RigVMVariableNode() {variable=eye leftovers}
   > &RigUnit_SetTransform(
    Weight=1
    bInitial=false
    bPropagateToChildren=true
    Item=@For_Each_1.Element
    Value=&RigUnit_GetTransform(
     bInitial=False
     Item=RigVMDispatchNode(Index=@For_Each_1.Index, Array=@n5.Value) {template="DISPATCH_RigVMDispatch_ArrayGetAtIndex(in Array,in Index,out Element)"}.Element
    ).Transform
   )
   RigVMDispatchNode(Index=0, Array=RigVMVariableNode() {variable=eye names}.Value) {template="DISPATCH_RigVMDispatch_ArrayGetAtIndex(in Array,in Index,out Element)"}
   Concat = RigVMUnitNode(A=(), B=()) {template="Concat::Execute(in A,in B,out Result)"}
   RigVMDispatchNode(B=@n2.Element.Name) {template="DISPATCH_RigVMDispatch_CoreEquals(in A,in B,out Result)"}
   RigVMVariableNode() {variable=eye null array}
  ]}
 New Function_ContainedGraph {nodes [
   n1 = RigVMFunctionEntryNode()
   RigVMFunction_Sequence = &RigVMFunction_Sequence(ExecuteContext=@n1.ExecuteContext)
   RigVMFunctionReturnNode(ExecuteContext=@RigVMFunction_Sequence.B)
   n2 = &RigVMFunction_ControlFlowBranch(
    Condition=&RigUnit_GetBoolAnimationChannel(Channel=Enable Cut Offset, Control=control rig options, bInitial=False) {template="GetAnimationChannel::Execute(out Value,in Control,in Channel,in bInitial)"}.Value
    ExecuteContext=@RigVMFunction_Sequence.A
   )
   RigVMFunction_Sequence_10 = RigVMAggregateNode(ExecuteContext=@n2.True)
   For_Each_14 = RigVMDispatchNode(Array=RigVMVariableNode() {variable=offset tentacle bones exclude dupes}.Value, ExecuteContext=@RigVMFunction_Sequence_10.A) {template="DISPATCH_RigVMDispatch_ArrayIterator(in Array,out Element,out Index,out Count,out Ratio)"}
   n3 = &RigUnit_GetTransform(
    bInitial=False
    Item=RigVMDispatchNode(Index=@For_Each_14.Index, Array=RigVMVariableNode() {variable=offset tentacle controls}.Value) {template="DISPATCH_RigVMDispatch_ArrayGetAtIndex(in Array,in Index,out Element)"}.Element
   )
   > &RigUnit_SetTransform(
    Value=(Rotation=(X=0,Y=0,Z=0,W=1),Translation=(X=0,Y=0,Z=0),Scale3D=(X=1,Y=1,Z=1))
    Weight=1
    bInitial=False
    Value.Rotation=@n3.Transform.Rotation
    Value.Translation=@n3.Transform.Translation
    Item=@For_Each_14.Element
   )
   > /Script/ControlRig.RigUnit_SetScale(Space=GlobalSpace, Weight=1, bInitial=False, bPropagateToChildren=false, Scale=@n3.Transform.Scale3D, Item=@For_Each_14.Element) {is=struct RigVMUnitNode}
   &RigUnit_SetTransform(
    Item=(Type=Bone,Name="tentacle_07_dup")
    Weight=1
    bInitial=False
    bPropagateToChildren=true
    Value=&RigUnit_GetTransform(Item=(Type=Control,Name="tentacle_07_dup_offset"), bInitial=False).Transform
    ExecutePin=@RigVMFunction_Sequence_10.C
   )
   > &RigUnit_SetTransform(
    Item=(Type=Bone,Name="tentacle_06_dup")
    Weight=1
    bInitial=False
    bPropagateToChildren=true
    Value=&RigUnit_GetTransform(Item=(Type=Control,Name="tentacle_06_dup_offset"), bInitial=False).Transform
   )
   > &RigUnit_SetTransform(
    Item=(Type=Bone,Name="tentacle_08_dup")
    Weight=1
    bInitial=False
    bPropagateToChildren=true
    Value=&RigUnit_GetTransform(Item=(Type=Control,Name="tentacle_08_dup_offset"), bInitial=False).Transform
   )
   n4 = /Script/ControlRig.RigUnit_IsInteracting() {is=struct RigVMUnitNode}
   For_Each_17 = RigVMDispatchNode(Array=@n4.Items, ExecuteContext=@RigVMFunction_Sequence_10.B) {template="DISPATCH_RigVMDispatch_ArrayIterator(in Array,out Element,out Index,out Count,out Ratio)"}
   n5 = RigVMVariableNode() {variable=offset tentacle controls}
   n6 = &RigVMFunction_ControlFlowBranch(
    ExecuteContext=@For_Each_17.ExecuteContext
    Condition=&RigVMFunction_MathBoolAnd(
     B=RigVMDispatchNode(Element=@For_Each_17.Element, Array=@n5.Value) {template="DISPATCH_RigVMDispatch_ArrayFind(in Array,in Element,out Index,out Success)"}.Success
     A=@n4.bIsInteracting
    ).Result
   )
   SetAnimationChannel = /Script/ControlRig.RigUnit_SetBoolAnimationChannel(Channel=Offsets follow FK, Control=control rig options, Value=False, bInitial=False, ExecutePin=@n6.True) {is=struct RigVMUnitNode, template="SetAnimationChannel::Execute(in Value,in Control,in Channel,in bInitial)"}
   n7 = &RigVMFunction_Sequence(ExecuteContext=@n2.False)
   n8 = &RigVMFunction_ControlFlowBranch(
    Condition=&RigUnit_GetBoolAnimationChannel(Channel=Offsets follow FK, Control=control rig options, bInitial=False) {template="GetAnimationChannel::Execute(out Value,in Control,in Channel,in bInitial)"}.Value
    ExecuteContext=@n7.A
   )
   For_Each_15 = RigVMDispatchNode(ExecuteContext=@n8.True, Array=RigVMVariableNode() {variable=upper tentacle controls}.Value) {template="DISPATCH_RigVMDispatch_ArrayIterator(in Array,out Element,out Index,out Count,out Ratio)"}
   > &RigUnit_SetControlOffset(
    Space=GlobalSpace
    Control=RigVMDispatchNode(Array=RigVMVariableNode() {variable=offset tentacle controls}.Value, Index=@For_Each_15.Index) {template="DISPATCH_RigVMDispatch_ArrayGetAtIndex(in Array,in Index,out Element)"}.Element.Name
    Offset=&RigUnit_GetTransform(bInitial=False, Item=@For_Each_15.Element).Transform
   ) {template="SetControlOffset::Execute(in Control,io Offset,in Space)"}
   &RigUnit_SetControlOffset(
    Control=tentacle_07_dup_offset
    Space=GlobalSpace
    Offset=&RigUnit_GetTransform(Item=(Type=Control,Name="tentacle_07_fk"), bInitial=False).Transform
    ExecutePin=@For_Each_15.Completed
   ) {template="SetControlOffset::Execute(in Control,io Offset,in Space)"}
   > &RigUnit_SetControlOffset(
    Control=tentacle_06_dup_offset
    Space=GlobalSpace
    Offset=&RigUnit_GetTransform(Item=(Type=Control,Name="tentacle_06_fk"), bInitial=False).Transform
   ) {template="SetControlOffset::Execute(in Control,io Offset,in Space)"}
   > &RigUnit_SetControlOffset(
    Control=tentacle_08_dup_offset
    Space=GlobalSpace
    Offset=&RigUnit_GetTransform(Item=(Type=Control,Name="tentacle_08_fk"), bInitial=False).Transform
   ) {template="SetControlOffset::Execute(in Control,io Offset,in Space)"}
   &RigUnit_SetTransform(
    Item=(Type=Bone,Name="tentacle_07_dup")
    Weight=1
    bInitial=False
    bPropagateToChildren=true
    Value=&RigUnit_GetTransform(Item=(Type=Control,Name="tentacle_07_fk"), bInitial=False).Transform
    ExecutePin=@n7.B
   )
   > &RigUnit_SetTransform(
    Item=(Type=Bone,Name="tentacle_06_dup")
    Weight=1
    bInitial=False
    bPropagateToChildren=true
    Value=&RigUnit_GetTransform(Item=(Type=Control,Name="tentacle_06_fk"), bInitial=False).Transform
   )
   > &RigUnit_SetTransform(
    Item=(Type=Bone,Name="tentacle_08_dup")
    Weight=1
    bInitial=False
    bPropagateToChildren=true
    Value=&RigUnit_GetTransform(Item=(Type=Control,Name="tentacle_08_dup_offset"), bInitial=False).Transform
   )
   RigVMVariableNode() {variable=upper tentacle controls}
  ]}
 New Function_ContainedGraph {nodes [
   n1 = RigVMFunctionEntryNode()
   RigVMFunction_Sequence = &RigVMFunction_Sequence(ExecuteContext=@n1.ExecuteContext)
   RigVMFunctionReturnNode(ExecuteContext=@RigVMFunction_Sequence.B)
   n2 = &RigVMFunction_Sequence(ExecuteContext=@RigVMFunction_Sequence.A)
   n3 = &RigUnit_HierarchyAddNull(
    Name=eyes
    Parent=(Type=Control,Name="root")
    Space=GlobalSpace
    Transform=(Rotation=(X=0,Y=0,Z=0,W=1),Translation=(X=0,Y=0,Z=0),Scale3D=(X=1,Y=1,Z=1))
    ExecutePin=@n2.A
   )
   > RigVMVariableNode(Value=@n3.Item) {variable=eyes null}
   > For_Each_6 = RigVMDispatchNode(Array=RigVMVariableNode() {variable=eye master array}.Value) {template="DISPATCH_RigVMDispatch_ArrayIterator(in Array,out Element,out Index,out Count,out Ratio)"}
   RigVMFunction_MathBoolNot = &RigVMFunction_MathBoolNot(Value=/Script/RigVM.RigVMFunction_Contains(Search=_mstr, Name=@For_Each_6.Element.Name) {is=struct RigVMUnitNode, template="Contains::Execute(in Name,in Search,out Result)"}.Result)
   RigVMFunction_MathBoolAnd = &RigVMFunction_MathBoolAnd(
    A=&RigVMFunction_Contains(Search=_innerlidA, Name=@For_Each_6.Element.Name) {template="Contains::Execute(in Name,in Search,out Result)"}.Result
    B=@RigVMFunction_MathBoolNot.Result
   )
   n4 = &RigVMFunction_ControlFlowBranch(ExecuteContext=@For_Each_6.ExecuteContext, Condition=@RigVMFunction_MathBoolAnd.Result)
   RigVMDispatchNode(ExecuteContext=@n4.True, Array=RigVMVariableNode() {variable=innerlidA array}.Value, Element=@For_Each_6.Element) {template="DISPATCH_RigVMDispatch_ArrayAdd(io Array,in Element,out Index)"}
   RigVMVariableNode(ExecuteContext=@n2.B) {variable=eye cycle count}
   > For_Each_4 = RigVMDispatchNode(Array=RigVMVariableNode() {variable=eye master array}.Value) {template="DISPATCH_RigVMDispatch_ArrayIterator(in Array,out Element,out Index,out Count,out Ratio)"}
   > RigVMVariableNode(Value=@For_Each_4.Element.Name) {variable=element name}
   Chop = /Script/RigVM.RigVMFunction_NameTruncate(Count=4, FromEnd=false, Name=@For_Each_4.Element.Name) {is=struct RigVMUnitNode, template="Chop::Execute(in Name,in Count,in FromEnd,out Remainder,out Chopped)"}
   > RigVMVariableNode(Value=@Chop.Chopped) {variable=eye prefix}
   > RigVMVariableNode(Value=@For_Each_4.Element) {variable=element}
   > n5 = RigVMVariableNode(Value=False) {variable=make status}
   n6 = &RigVMFunction_Sequence(ExecuteContext=@n5.ExecuteContext)
   n7 = RigVMVariableNode() {variable=eye names}
   n8 = RigVMVariableNode() {variable=eye cycle count}
   n9 = &RigVMFunction_ControlFlowBranch(
    Condition=RigVMDispatchNode(
     B=RigVMDispatchNode(Array=@n7.Value, Index=@n8.Value) {template="DISPATCH_RigVMDispatch_ArrayGetAtIndex(in Array,in Index,out Element)"}.Element
     A=RigVMVariableNode() {variable=eye prefix}.Value
    ) {template="DISPATCH_RigVMDispatch_CoreEquals(in A,in B,out Result)"}.Result
    ExecuteContext=@n6.A
   )
   RigVMVariableNode(
    Value=/Script/RigVM.RigVMFunction_MathIntAdd(B=1, A=RigVMVariableNode() {variable=eye cycle count}.Value) {is=struct RigVMUnitNode, template="Add::Execute(in A,in B,out Result)"}.Result
    ExecuteContext=@n9.False
   ) {variable=eye cycle count}
   n10 = &RigVMFunction_Sequence(ExecuteContext=@n6.B)
   n11 = &RigVMFunction_Sequence(ExecuteContext=@n10.B)
   n12 = &RigVMFunction_Sequence(ExecuteContext=@n11.B)
   n13 = &RigVMFunction_Sequence(ExecuteContext=@n12.B)
   n14 = &RigVMFunction_Sequence(ExecuteContext=@n13.B)
   n15 = &RigVMFunction_Sequence(ExecuteContext=@n14.B)
   n16 = &RigVMFunction_ControlFlowBranch(
    Condition=&RigVMFunction_Contains(Search=innerlid_mstr, Name=RigVMVariableNode() {variable=element name}.Value) {template="Contains::Execute(in Name,in Search,out Result)"}.Result
    ExecuteContext=@n15.A
   )
   n17 = RigVMDispatchNode(Array=RigVMVariableNode() {variable=innerlidA array}.Value, Index=RigVMVariableNode() {variable=eye cycle count}.Value) {template="DISPATCH_RigVMDispatch_ArrayGetAtIndex(in Array,in Index,out Element)"}
   n18 = &RigUnit_GetTransform(bInitial=False, Item=RigVMVariableNode() {variable=element}.Value)
   n19 = /Script/ControlRig.RigUnit_HierarchyAddControlTransform(
    InitialValue=(Rotation=(X=0,Y=0,Z=0,W=1),Translation=(X=0,Y=0,Z=0),Scale3D=(X=1,Y=1,Z=1))
    OffsetSpace=GlobalSpace
    OffsetTransform=(Rotation=(X=0,Y=0,Z=0,W=1),Translation=(X=0,Y=0,Z=0),Scale3D=(X=1,Y=1,Z=1))
    Settings="(InitialSpace=LocalSpace,bUsePreferredRotationOrder=False,PreferredRotationOrder=YZX,Limits=(LimitTranslationX=(bMinimum=False,bMaximum=False),LimitTranslationY=(bMinimum=False,bMaximum=False),LimitTranslationZ=(bMinimum=False,bMaximum=False),LimitPitch=(b ...(truncated)"
    Settings.Shape.Transform.Translation.X=&RigVMFunction_MathDoubleAdd(B=100, A=&RigUnit_GetTransform(Space=LocalSpace, bInitial=False, Item=@n17.Element).Transform.Translation.X).Result
    OffsetTransform.Rotation=@n18.Transform.Rotation
    OffsetTransform.Translation=@n18.Transform.Translation
    ExecutePin=@n16.True
    Name=RigVMVariableNode() {variable=element name}.Value
    Parent=RigVMVariableNode() {variable=orb parent}.Value
   ) {
    is=struct RigVMUnitNode
    template="SpawnControl::Execute(in InitialValue,in Settings,in OffsetTransform,in OffsetSpace,in Parent,in Name,out Item)"
   }
   n20 = &RigVMFunction_Sequence(ExecuteContext=@n19.ExecutePin)
   RigVMDispatchNode(ExecuteContext=@n20.A, Array=RigVMVariableNode() {variable=eye control array}.Value, Element=@n19.Item) {template="DISPATCH_RigVMDispatch_ArrayAdd(io Array,in Element,out Index)"}
   > RigVMDispatchNode(Array=RigVMVariableNode() {variable=innerlid control array}.Value, Element=@n19.Item) {template="DISPATCH_RigVMDispatch_ArrayAdd(io Array,in Element,out Index)"}
   n21 = RigVMVariableNode() {variable=element}
   RigVMDispatchNode(Array=RigVMVariableNode() {variable=eye matching skel array}.Value, Element=@n21.Value, ExecuteContext=@n20.B) {template="DISPATCH_RigVMDispatch_ArrayAdd(io Array,in Element,out Index)"}
   > RigVMDispatchNode(Array=RigVMVariableNode() {variable=innerlid skel array}.Value, Element=@n21.Value) {template="DISPATCH_RigVMDispatch_ArrayAdd(io Array,in Element,out Index)"}
   > RigVMVariableNode(Value=true) {variable=make status}
   RigVMFunction_ControlFlowBranch = &RigVMFunction_ControlFlowBranch(Condition=RigVMVariableNode() {variable=make status}.Value, ExecuteContext=@n15.B)
   n22 = RigVMVariableNode() {variable=element}
   n23 = &RigUnit_GetTransform(bInitial=False, Item=@n22.Value)
   SpawnControl = &RigUnit_HierarchyAddControlTransform(
    OffsetSpace=GlobalSpace
    Parent=(Type=Control,Name="None")
    Settings="(InitialSpace=GlobalSpace,bUsePreferredRotationOrder=False,PreferredRotationOrder=YZX,Limits=(LimitTranslationX=(bMinimum=False,bMaximum=False),LimitTranslationY=(bMinimum=False,bMaximum=False),LimitTranslationZ=(bMinimum=False,bMaximum=False),LimitPitch=( ...(truncated)"
    InitialValue=@n23.Transform
    OffsetTransform=@n23.Transform
    Parent.Name=/Script/ControlRig.RigUnit_HierarchyGetParent(bDefaultParent=True, Child=@n22.Value) {is=struct RigVMUnitNode}.Parent.Name
    Name=RigVMVariableNode() {variable=element name}.Value
    ExecutePin=@RigVMFunction_ControlFlowBranch.False
   ) {template="SpawnControl::Execute(in InitialValue,in Settings,in OffsetTransform,in OffsetSpace,in Parent,in Name,out Item)"}
   > Add = RigVMDispatchNode(Element=@SpawnControl.Item, Array=RigVMVariableNode() {variable=eye leftovers}.Value) {template="DISPATCH_RigVMDispatch_ArrayAdd(io Array,in Element,out Index)"}
   > RigVMDispatchNode(Array=RigVMVariableNode() {variable=eye leftovers joints}.Value, Element=RigVMVariableNode() {variable=element}.Value) {template="DISPATCH_RigVMDispatch_ArrayAdd(io Array,in Element,out Index)"}
   n24 = &RigVMFunction_ControlFlowBranch(
    Condition=&RigVMFunction_Contains(Search=_ball, Name=RigVMVariableNode() {variable=element name}.Value) {template="Contains::Execute(in Name,in Search,out Result)"}.Result
    ExecuteContext=@n14.A
   )
   n25 = &RigUnit_GetTransform(bInitial=False, Item=RigVMVariableNode() {variable=element}.Value)
   n26 = &RigUnit_HierarchyAddControlTransform(
    InitialValue=(Rotation=(X=0,Y=0,Z=0,W=1),Translation=(X=0,Y=0,Z=0),Scale3D=(X=1,Y=1,Z=1))
    OffsetSpace=GlobalSpace
    OffsetTransform=(Rotation=(X=0,Y=0,Z=0,W=1),Translation=(X=0,Y=0,Z=0),Scale3D=(X=1,Y=1,Z=1))
    Settings="(InitialSpace=LocalSpace,bUsePreferredRotationOrder=False,PreferredRotationOrder=YZX,Limits=(LimitTranslationX=(bMinimum=False,bMaximum=False),LimitTranslationY=(bMinimum=False,bMaximum=False),LimitTranslationZ=(bMinimum=False,bMaximum=False),LimitPitch=(b ...(truncated)"
    Settings.Shape.Transform.Translation.X=RigVMDispatchNode(Array=RigVMVariableNode() {variable=eye orb offsets}.Value, Index=RigVMVariableNode() {variable=eye cycle count}.Value) {template="DISPATCH_RigVMDispatch_ArrayGetAtIndex(in Array,in Index,out Element)"}.Element
    OffsetTransform.Rotation=@n25.Transform.Rotation
    OffsetTransform.Translation=@n25.Transform.Translation
    ExecutePin=@n24.True
    Name=RigVMVariableNode() {variable=element name}.Value
    Parent=RigVMVariableNode() {variable=orb parent}.Value
   ) {template="SpawnControl::Execute(in InitialValue,in Settings,in OffsetTransform,in OffsetSpace,in Parent,in Name,out Item)"}
   n27 = &RigVMFunction_Sequence(ExecuteContext=@n26.ExecutePin)
   RigVMDispatchNode(ExecuteContext=@n27.A, Array=RigVMVariableNode() {variable=eye control array}.Value, Element=@n26.Item) {template="DISPATCH_RigVMDispatch_ArrayAdd(io Array,in Element,out Index)"}
   > RigVMDispatchNode(Array=RigVMVariableNode() {variable=eye ball control array}.Value, Element=@n26.Item) {template="DISPATCH_RigVMDispatch_ArrayAdd(io Array,in Element,out Index)"}
   RigVMDispatchNode(
    ExecuteContext=@n27.B
    Array=RigVMVariableNode() {variable=eye matching skel array}.Value
    Element=RigVMVariableNode() {variable=element}.Value
   ) {template="DISPATCH_RigVMDispatch_ArrayAdd(io Array,in Element,out Index)"}
   > RigVMVariableNode(Value=true) {variable=make status}
   n28 = &RigVMFunction_Sequence(ExecuteContext=@n13.A)
   n29 = RigVMVariableNode() {variable=element name}
   n30 = &RigVMFunction_ControlFlowBranch(
    Condition=&RigVMFunction_MathBoolOr(
     A=&RigVMFunction_Contains(Search=_lid_upperA, Name=@n29.Value) {template="Contains::Execute(in Name,in Search,out Result)"}.Result
     B=&RigVMFunction_Contains(Search=_lid_upperB, Name=@n29.Value) {template="Contains::Execute(in Name,in Search,out Result)"}.Result
    ).Result
    ExecuteContext=@n28.B
   )
   n31 = &RigUnit_HierarchyAddControlTransform(
    InitialValue=(Rotation=(X=0,Y=0,Z=0,W=1),Translation=(X=0,Y=0,Z=0),Scale3D=(X=1,Y=1,Z=1))
    OffsetSpace=GlobalSpace
    Settings="(InitialSpace=LocalSpace,bUsePreferredRotationOrder=False,PreferredRotationOrder=YZX,Limits=(LimitTranslationX=(bMinimum=False,bMaximum=False),LimitTranslationY=(bMinimum=False,bMaximum=False),LimitTranslationZ=(bMinimum=False,bMaximum=False),LimitPitch=(b ...(truncated)"
    OffsetTransform=&RigUnit_GetTransform(bInitial=False, Item=RigVMVariableNode() {variable=element}.Value).Transform
    ExecutePin=@n30.True
    Name=RigVMVariableNode() {variable=element name}.Value
    Parent=RigVMVariableNode() {variable=upper lid parent}.Value
   ) {template="SpawnControl::Execute(in InitialValue,in Settings,in OffsetTransform,in OffsetSpace,in Parent,in Name,out Item)"}
   n32 = &RigVMFunction_Sequence(ExecuteContext=@n31.ExecutePin)
   RigVMDispatchNode(ExecuteContext=@n32.A, Array=RigVMVariableNode() {variable=eye control array}.Value, Element=@n31.Item) {template="DISPATCH_RigVMDispatch_ArrayAdd(io Array,in Element,out Index)"}
   > RigVMDispatchNode(Array=RigVMVariableNode() {variable=lid control array}.Value, Element=@n31.Item) {template="DISPATCH_RigVMDispatch_ArrayAdd(io Array,in Element,out Index)"}
   RigVMDispatchNode(
    ExecuteContext=@n32.B
    Array=RigVMVariableNode() {variable=eye matching skel array}.Value
    Element=RigVMVariableNode() {variable=element}.Value
   ) {template="DISPATCH_RigVMDispatch_ArrayAdd(io Array,in Element,out Index)"}
   > RigVMVariableNode(Value=true) {variable=make status}
   n33 = RigVMVariableNode() {variable=element name}
   RigVMFunction_MathBoolOr = &RigVMFunction_MathBoolOr(
    A=&RigVMFunction_Contains(Search=_lid_lowerB, Name=@n33.Value) {template="Contains::Execute(in Name,in Search,out Result)"}.Result
    B=&RigVMFunction_Contains(Search=_lid_lowerA, Name=@n33.Value) {template="Contains::Execute(in Name,in Search,out Result)"}.Result
   )
   n34 = &RigVMFunction_ControlFlowBranch(Condition=@RigVMFunction_MathBoolOr.Result, ExecuteContext=@n28.A)
   n35 = &RigUnit_HierarchyAddControlTransform(
    InitialValue=(Rotation=(X=0,Y=0,Z=0,W=1),Translation=(X=0,Y=0,Z=0),Scale3D=(X=1,Y=1,Z=1))
    OffsetSpace=GlobalSpace
    Settings="(InitialSpace=LocalSpace,bUsePreferredRotationOrder=False,PreferredRotationOrder=YZX,Limits=(LimitTranslationX=(bMinimum=False,bMaximum=False),LimitTranslationY=(bMinimum=False,bMaximum=False),LimitTranslationZ=(bMinimum=False,bMaximum=False),LimitPitch=(b ...(truncated)"
    OffsetTransform=&RigUnit_GetTransform(bInitial=False, Item=RigVMVariableNode() {variable=element}.Value).Transform
    ExecutePin=@n34.True
    Name=RigVMVariableNode() {variable=element name}.Value
    Parent=RigVMVariableNode() {variable=lower lid parent}.Value
   ) {template="SpawnControl::Execute(in InitialValue,in Settings,in OffsetTransform,in OffsetSpace,in Parent,in Name,out Item)"}
   n36 = &RigVMFunction_Sequence(ExecuteContext=@n35.ExecutePin)
   RigVMDispatchNode(ExecuteContext=@n36.A, Array=RigVMVariableNode() {variable=eye control array}.Value, Element=@n35.Item) {template="DISPATCH_RigVMDispatch_ArrayAdd(io Array,in Element,out Index)"}
   > RigVMDispatchNode(Array=RigVMVariableNode() {variable=lid control array}.Value, Element=@n35.Item) {template="DISPATCH_RigVMDispatch_ArrayAdd(io Array,in Element,out Index)"}
   RigVMDispatchNode(
    ExecuteContext=@n36.B
    Array=RigVMVariableNode() {variable=eye matching skel array}.Value
    Element=RigVMVariableNode() {variable=element}.Value
   ) {template="DISPATCH_RigVMDispatch_ArrayAdd(io Array,in Element,out Index)"}
   > RigVMVariableNode(Value=true) {variable=make status}
   n37 = &RigVMFunction_Sequence(ExecuteContext=@n12.A)
   n38 = &RigVMFunction_ControlFlowBranch(
    Condition=&RigVMFunction_Contains(Search=_lid_upper_master, Name=RigVMVariableNode() {variable=element name}.Value) {template="Contains::Execute(in Name,in Search,out Result)"}.Result
    ExecuteContext=@n37.A
   )
   n39 = &RigUnit_HierarchyAddControlTransform(
    InitialValue=(Rotation=(X=0,Y=0,Z=0,W=1),Translation=(X=0,Y=0,Z=0),Scale3D=(X=1,Y=1,Z=1))
    OffsetSpace=GlobalSpace
    Settings="(InitialSpace=LocalSpace,bUsePreferredRotationOrder=False,PreferredRotationOrder=YZX,Limits=(LimitTranslationX=(bMinimum=False,bMaximum=False),LimitTranslationY=(bMinimum=False,bMaximum=False),LimitTranslationZ=(bMinimum=False,bMaximum=False),LimitPitch=(b ...(truncated)"
    OffsetTransform=&RigUnit_GetTransform(bInitial=False, Item=RigVMVariableNode() {variable=element}.Value).Transform
    ExecutePin=@n38.True
    Parent=RigVMVariableNode() {variable=running parent}.Value
    Name=RigVMVariableNode() {variable=element name}.Value
   ) {template="SpawnControl::Execute(in InitialValue,in Settings,in OffsetTransform,in OffsetSpace,in Parent,in Name,out Item)"}
   n40 = &RigVMFunction_Sequence(ExecuteContext=@n39.ExecutePin)
   RigVMVariableNode(ExecuteContext=@n40.A, Value=@n39.Item) {variable=upper lid parent}
   > RigVMDispatchNode(Array=RigVMVariableNode() {variable=eye control array}.Value, Element=@n39.Item) {template="DISPATCH_RigVMDispatch_ArrayAdd(io Array,in Element,out Index)"}
   > RigVMDispatchNode(Array=RigVMVariableNode() {variable=lid control array}.Value, Element=@n39.Item) {template="DISPATCH_RigVMDispatch_ArrayAdd(io Array,in Element,out Index)"}
   RigVMDispatchNode(
    ExecuteContext=@n40.B
    Array=RigVMVariableNode() {variable=eye matching skel array}.Value
    Element=RigVMVariableNode() {variable=element}.Value
   ) {template="DISPATCH_RigVMDispatch_ArrayAdd(io Array,in Element,out Index)"}
   > RigVMVariableNode(Value=true) {variable=make status}
   n41 = &RigVMFunction_ControlFlowBranch(
    Condition=&RigVMFunction_Contains(Search=_lid_lower_master, Name=RigVMVariableNode() {variable=element name}.Value) {template="Contains::Execute(in Name,in Search,out Result)"}.Result
    ExecuteContext=@n37.B
   )
   n42 = &RigUnit_HierarchyAddControlTransform(
    InitialValue=(Rotation=(X=0,Y=0,Z=0,W=1),Translation=(X=0,Y=0,Z=0),Scale3D=(X=1,Y=1,Z=1))
    OffsetSpace=GlobalSpace
    Settings="(InitialSpace=LocalSpace,bUsePreferredRotationOrder=False,PreferredRotationOrder=YZX,Limits=(LimitTranslationX=(bMinimum=False,bMaximum=False),LimitTranslationY=(bMinimum=False,bMaximum=False),LimitTranslationZ=(bMinimum=False,bMaximum=False),LimitPitch=(b ...(truncated)"
    OffsetTransform=&RigUnit_GetTransform(bInitial=False, Item=RigVMVariableNode() {variable=element}.Value).Transform
    ExecutePin=@n41.True
    Parent=RigVMVariableNode() {variable=running parent}.Value
    Name=RigVMVariableNode() {variable=element name}.Value
   ) {template="SpawnControl::Execute(in InitialValue,in Settings,in OffsetTransform,in OffsetSpace,in Parent,in Name,out Item)"}
   n43 = &RigVMFunction_Sequence(ExecuteContext=@n42.ExecutePin)
   RigVMVariableNode(ExecuteContext=@n43.A, Value=@n42.Item) {variable=lower lid parent}
   > RigVMDispatchNode(Array=RigVMVariableNode() {variable=eye control array}.Value, Element=@n42.Item) {template="DISPATCH_RigVMDispatch_ArrayAdd(io Array,in Element,out Index)"}
   > RigVMDispatchNode(Array=RigVMVariableNode() {variable=lid control array}.Value, Element=@n42.Item) {template="DISPATCH_RigVMDispatch_ArrayAdd(io Array,in Element,out Index)"}
   RigVMDispatchNode(
    ExecuteContext=@n43.B
    Array=RigVMVariableNode() {variable=eye matching skel array}.Value
    Element=RigVMVariableNode() {variable=element}.Value
   ) {template="DISPATCH_RigVMDispatch_ArrayAdd(io Array,in Element,out Index)"}
   > RigVMVariableNode(Value=true) {variable=make status}
   n44 = &RigVMFunction_ControlFlowBranch(
    Condition=&RigVMFunction_Contains(Search=_orb, Name=RigVMVariableNode() {variable=element name}.Value) {template="Contains::Execute(in Name,in Search,out Result)"}.Result
    ExecuteContext=@n11.A
   )
   n45 = &RigUnit_GetTransform(bInitial=False, Item=RigVMVariableNode() {variable=element}.Value)
   n46 = RigVMVariableNode() {variable=eye cycle count}
   n47 = &RigUnit_HierarchyAddControlTransform(
    InitialValue=(Rotation=(X=0,Y=0,Z=0,W=1),Translation=(X=0,Y=0,Z=0),Scale3D=(X=1,Y=1,Z=1))
    OffsetSpace=GlobalSpace
    OffsetTransform=(Rotation=(X=0,Y=-0.7071,Z=0,W=0.7071),Translation=(X=0,Y=0,Z=0),Scale3D=(X=1,Y=1,Z=1))
    Settings="(InitialSpace=LocalSpace,bUsePreferredRotationOrder=False,PreferredRotationOrder=YZX,Limits=(LimitTranslationX=(bMinimum=False,bMaximum=False),LimitTranslationY=(bMinimum=False,bMaximum=False),LimitTranslationZ=(bMinimum=False,bMaximum=False),LimitPitch=(b ...(truncated)"
    Name=/Script/RigVM.RigVMFunction_NameConcat(B=_orb, A=RigVMVariableNode() {variable=eye prefix}.Value) {is=struct RigVMUnitNode, template="Concat::Execute(in A,in B,out Result)"}.Result
    OffsetTransform.Rotation=@n45.Transform.Rotation
    OffsetTransform.Translation=@n45.Transform.Translation
    Settings.Shape.Transform.Translation.X=RigVMDispatchNode(
     False=550
     True=300
     Condition=RigVMDispatchNode(B=0, A=@n46.Value) {template="DISPATCH_RigVMDispatch_CoreEquals(in A,in B,out Result)"}.Result
    ) {template="DISPATCH_RigVMDispatch_If(in Condition,in True,in False,out Result)"}.Result
    ExecutePin=@n44.True
    Parent=RigVMVariableNode() {variable=running parent}.Value
   ) {template="SpawnControl::Execute(in InitialValue,in Settings,in OffsetTransform,in OffsetSpace,in Parent,in Name,out Item)"}
   n48 = &RigVMFunction_Sequence(ExecuteContext=@n47.ExecutePin)
   RigVMVariableNode(ExecuteContext=@n48.A, Value=@n47.Item) {variable=orb parent}
   > RigVMDispatchNode(Array=RigVMVariableNode() {variable=eye control array}.Value, Element=@n47.Item) {template="DISPATCH_RigVMDispatch_ArrayAdd(io Array,in Element,out Index)"}
   > RigVMDispatchNode(Array=RigVMVariableNode() {variable=eye orb controls}.Value, Element=@n47.Item) {template="DISPATCH_RigVMDispatch_ArrayAdd(io Array,in Element,out Index)"}
   n49 = RigVMVariableNode() {variable=element}
   RigVMDispatchNode(Array=RigVMVariableNode() {variable=eye matching skel array}.Value, Element=@n49.Value, ExecuteContext=@n48.B) {template="DISPATCH_RigVMDispatch_ArrayAdd(io Array,in Element,out Index)"}
   > RigVMDispatchNode(Array=RigVMVariableNode() {variable=eye orb array}.Value, Element=@n49.Value) {template="DISPATCH_RigVMDispatch_ArrayAdd(io Array,in Element,out Index)"}
   > RigVMVariableNode(Value=true) {variable=make status}
   n50 = RigVMVariableNode() {variable=eye names}
   n51 = RigVMVariableNode() {variable=eye cycle count}
   n52 = &RigVMFunction_ControlFlowBranch(
    Condition=RigVMDispatchNode(
     B=RigVMDispatchNode(Array=@n50.Value, Index=@n51.Value) {template="DISPATCH_RigVMDispatch_ArrayGetAtIndex(in Array,in Index,out Element)"}.Element
     A=RigVMVariableNode() {variable=element name}.Value
    ) {template="DISPATCH_RigVMDispatch_CoreEquals(in A,in B,out Result)"}.Result
    ExecuteContext=@n10.A
   )
   n53 = &RigUnit_ChainInfo(
    DebugScale=1
    Items=((Type=None,Name="None"),(Type=None,Name="None"),(Type=None,Name="None"))
    bCalculateStretch=True
    bDebug=false
    bInitial=False
    Items.0=RigVMDispatchNode(Array=RigVMVariableNode() {variable=eye interp start bones}.Value, Index=@n51.Value) {template="DISPATCH_RigVMDispatch_ArrayGetAtIndex(in Array,in Index,out Element)"}.Element
    Items.2=RigVMDispatchNode(Array=RigVMVariableNode() {variable=eye interp end bones}.Value, Index=@n51.Value) {template="DISPATCH_RigVMDispatch_ArrayGetAtIndex(in Array,in Index,out Element)"}.Element
    Param=RigVMDispatchNode(Array=RigVMVariableNode() {variable=eye interp param}.Value, Index=@n51.Value) {template="DISPATCH_RigVMDispatch_ArrayGetAtIndex(in Array,in Index,out Element)"}.Element
    Items.1=RigVMDispatchNode(Array=RigVMVariableNode() {variable=eye interp mid bones}.Value, Index=@n51.Value) {template="DISPATCH_RigVMDispatch_ArrayGetAtIndex(in Array,in Index,out Element)"}.Element
   )
   n54 = &RigUnit_HierarchyAddNull(
    Space=GlobalSpace
    Transform=@n53.InterpolatedTransform
    Name=&RigVMFunction_NameConcat(B=_null, A=RigVMVariableNode() {variable=eye prefix}.Value) {template="Concat::Execute(in A,in B,out Result)"}.Result
    ExecutePin=@n52.True
    Parent=RigVMVariableNode() {variable=eyes null}.Value
   )
   > RigVMDispatchNode(Array=RigVMVariableNode() {variable=eye null array}.Value, Element=@n54.Item) {template="DISPATCH_RigVMDispatch_ArrayAdd(io Array,in Element,out Index)"}
   > n55 = &RigUnit_HierarchyAddControlTransform(
    InitialValue=(Rotation=(X=0,Y=0,Z=0,W=1),Translation=(X=0,Y=0,Z=0),Scale3D=(X=1,Y=1,Z=1))
    OffsetSpace=GlobalSpace
    Settings="(InitialSpace=LocalSpace,bUsePreferredRotationOrder=False,PreferredRotationOrder=YZX,Limits=(LimitTranslationX=(bMinimum=False,bMaximum=False),LimitTranslationY=(bMinimum=False,bMaximum=False),LimitTranslationZ=(bMinimum=False,bMaximum=False),LimitPitch=(b ...(truncated)"
    OffsetTransform=&RigUnit_GetTransform(bInitial=False, Item=RigVMVariableNode() {variable=element}.Value).Transform
    Name=RigVMVariableNode() {variable=element name}.Value
    Parent=@n54.Item
   ) {template="SpawnControl::Execute(in InitialValue,in Settings,in OffsetTransform,in OffsetSpace,in Parent,in Name,out Item)"}
   n56 = &RigVMFunction_Sequence(ExecuteContext=@n55.ExecutePin)
   RigVMVariableNode(ExecuteContext=@n56.A, Value=@n55.Item) {variable=running parent}
   > RigVMDispatchNode(Array=RigVMVariableNode() {variable=eye control array}.Value, Element=@n55.Item) {template="DISPATCH_RigVMDispatch_ArrayAdd(io Array,in Element,out Index)"}
   > RigVMDispatchNode(Array=RigVMVariableNode() {variable=eye master controls}.Value, Element=@n55.Item) {template="DISPATCH_RigVMDispatch_ArrayAdd(io Array,in Element,out Index)"}
   n57 = RigVMVariableNode() {variable=element}
   > RigVMDispatchNode(
    Element=RigVMDispatchNode(
     Index=0
     Array=/Script/ControlRig.RigUnit_CollectionChildrenArray(TypeToSearch=All, bDefaultChildren=True, bIncludeParent=False, bRecursive=False, Parent=@n57.Value) {is=struct RigVMUnitNode}.Items
    ) {template="DISPATCH_RigVMDispatch_ArrayGetAtIndex(in Array,in Index,out Element)"}.Element
    Array=RigVMVariableNode() {variable=eye matching skel array}.Value
   ) {template="DISPATCH_RigVMDispatch_ArrayAdd(io Array,in Element,out Index)"}
   RigVMVariableNode(Value=true, ExecuteContext=@n56.B) {variable=make status}
   > n58 = /Script/ControlRig.RigUnit_HierarchyAddAnimationChannelFloat(
    InitialValue=0
    LimitsEnabled=(Enabled=(bMinimum=True,bMaximum=True))
    MaximumValue=1
    MinimumValue=0
    Name=&RigVMFunction_NameConcat(B=" inner blink", A=RigVMVariableNode() {variable=eye prefix}.Value) {template="Concat::Execute(in A,in B,out Result)"}.Result
    Parent=@n55.Item
   ) {
    is=struct RigVMUnitNode
    template="SpawnAnimationChannel::Execute(in InitialValue,in MinimumValue,in MaximumValue,in LimitsEnabled,in Parent,in Name,out Item)"
   }
   > RigVMDispatchNode(Element=@n58.Item, Array=RigVMVariableNode() {variable=blink channels}.Value) {template="DISPATCH_RigVMDispatch_ArrayAdd(io Array,in Element,out Index)"}
   > &RigUnit_HierarchyAddAnimationChannelFloat(
    InitialValue=0
    LimitsEnabled=(Enabled=(bMinimum=True,bMaximum=True))
    MaximumValue=1
    MinimumValue=0
    Name=&RigVMFunction_NameConcat(B=" blink", A=RigVMVariableNode() {variable=eye prefix}.Value) {template="Concat::Execute(in A,in B,out Result)"}.Result
    Parent=@n55.Item
   ) {template="SpawnAnimationChannel::Execute(in InitialValue,in MinimumValue,in MaximumValue,in LimitsEnabled,in Parent,in Name,out Item)"}
  ]}
 New Function_ContainedGraph {nodes [
   n1 = RigVMFunctionEntryNode()
   RigVMFunction_Sequence = &RigVMFunction_Sequence(ExecuteContext=@n1.ExecuteContext)
   RigVMFunctionReturnNode(ExecuteContext=@RigVMFunction_Sequence.B)
   RigVMFunction_Sequence_1 = RigVMAggregateNode(ExecuteContext=@RigVMFunction_Sequence.A)
   n2 = /Script/ControlRig.RigUnit_GetFloatAnimationChannelFromItem(bInitial=False, Item=RigVMVariableNode() {variable=cut constraints positional channel}.Value) {is=struct RigVMUnitNode, template="GetAnimationChannelFromItem::Execute(out Value,in Item,in bInitial)"}
   /Script/ControlRig.RigUnit_PositionConstraintLocalSpaceOffset(
    Child=(Type=Bone,Name="tentacle_07_dup")
    Filter=(bX=True,bY=True,bZ=True)
    Parents=((Item=(Type=Bone,Name="tentacle_07"),Weight=1))
    bMaintainOffset=True
    ExecutePin=@RigVMFunction_Sequence_1.B
    Weight=@n2.Value
   ) {is=struct RigVMUnitNode}
   > &RigUnit_PositionConstraintLocalSpaceOffset(
    Child=(Type=Bone,Name="tentacle_06_dup")
    Filter=(bX=True,bY=True,bZ=True)
    Parents=((Item=(Type=Bone,Name="tentacle_06"),Weight=1))
    bMaintainOffset=True
    Weight=@n2.Value
   )
   > &RigUnit_PositionConstraintLocalSpaceOffset(
    Child=(Type=Bone,Name="tentacle_08_dup")
    Filter=(bX=True,bY=True,bZ=True)
    Parents=((Item=(Type=Bone,Name="tentacle_08"),Weight=1))
    bMaintainOffset=True
    Weight=@n2.Value
   )
   n3 = &RigUnit_GetFloatAnimationChannelFromItem(bInitial=False, Item=RigVMVariableNode() {variable=cut constraints rotational channel}.Value) {template="GetAnimationChannelFromItem::Execute(out Value,in Item,in bInitial)"}
   &RigUnit_RotationConstraintLocalSpaceOffset(
    AdvancedSettings=(InterpolationType=Average,RotationOrderForFilter=XZY)
    Child=(Type=Bone,Name="tentacle_07_dup")
    Filter=(bX=True,bY=True,bZ=True)
    Parents=((Item=(Type=Bone,Name="tentacle_07"),Weight=1))
    ExecutePin=@RigVMFunction_Sequence_1.A
    Weight=@n3.Value
   )
   > &RigUnit_RotationConstraintLocalSpaceOffset(
    AdvancedSettings=(InterpolationType=Average,RotationOrderForFilter=XZY)
    Child=(Type=Bone,Name="tentacle_06_dup")
    Filter=(bX=True,bY=True,bZ=True)
    Parents=((Item=(Type=Bone,Name="tentacle_06"),Weight=1))
    Weight=@n3.Value
   )
   > &RigUnit_RotationConstraintLocalSpaceOffset(
    AdvancedSettings=(InterpolationType=Average,RotationOrderForFilter=XZY)
    Child=(Type=Bone,Name="tentacle_08_dup")
    Filter=(bX=True,bY=True,bZ=True)
    Parents=((Item=(Type=Bone,Name="tentacle_08"),Weight=1))
    Weight=@n3.Value
   )
   &RigUnit_RotationConstraintLocalSpaceOffset(
    AdvancedSettings=(InterpolationType=Average,RotationOrderForFilter=XZY)
    Child=(Type=Bone,Name="eyeA_innerlidB_adj")
    Filter=(bX=True,bY=True,bZ=True)
    Parents=((Item=(Type=Bone,Name="eyeA_innerlid_mstr"),Weight=0.5),(Item=(Type=Bone,Name="eyeA_orb"),Weight=0.5))
    Weight=1
    ExecutePin=@RigVMFunction_Sequence_1.C
   )
   > &RigUnit_RotationConstraintLocalSpaceOffset(
    AdvancedSettings=(InterpolationType=Average,RotationOrderForFilter=XZY)
    Child=(Type=Bone,Name="eyeB_innerlidB_adj")
    Filter=(bX=True,bY=True,bZ=True)
    Parents=((Item=(Type=Bone,Name="eyeB_innerlid_mstr"),Weight=0.5),(Item=(Type=Bone,Name="eyeB_orb"),Weight=0.5))
    Weight=1
   )
   > &RigUnit_RotationConstraintLocalSpaceOffset(
    AdvancedSettings=(InterpolationType=Average,RotationOrderForFilter=XZY)
    Child=(Type=Bone,Name="eyeC_innerlidB_mstr")
    Filter=(bX=True,bY=True,bZ=True)
    Parents=((Item=(Type=Bone,Name="eyeC_innerlid_mstr"),Weight=0.5),(Item=(Type=Bone,Name="eyeC_orb"),Weight=0.5))
    Weight=1
   )
   > &RigUnit_RotationConstraintLocalSpaceOffset(
    AdvancedSettings=(InterpolationType=Average,RotationOrderForFilter=XZY)
    Child=(Type=Bone,Name="eyeD_innerlidB_adj")
    Filter=(bX=True,bY=True,bZ=True)
    Parents=((Item=(Type=Bone,Name="eyeD_innerlid_mstr"),Weight=0.5),(Item=(Type=Bone,Name="eyeD_orb"),Weight=0.5))
    Weight=1
   )
   > &RigUnit_RotationConstraintLocalSpaceOffset(
    AdvancedSettings=(InterpolationType=Average,RotationOrderForFilter=XZY)
    Child=(Type=Bone,Name="eyeE_innerlidB_mstr")
    Filter=(bX=True,bY=True,bZ=True)
    Parents=((Item=(Type=Bone,Name="eyeE_innerlid_mstr"),Weight=0.5),(Item=(Type=Bone,Name="eyeE_orb"),Weight=0.5))
    Weight=1
   )
  ]}
 New Function_ContainedGraph {nodes [
   n1 = RigVMFunctionEntryNode()
   RigVMFunction_Sequence = &RigVMFunction_Sequence(ExecuteContext=@n1.ExecuteContext)
   RigVMFunctionReturnNode(ExecuteContext=@RigVMFunction_Sequence.B)
   RigVMFunction_Sequence_2 = RigVMAggregateNode(ExecuteContext=@RigVMFunction_Sequence.A)
   n2 = &RigVMFunction_Sequence(ExecuteContext=@RigVMFunction_Sequence_2.C)
   Clone = RigVMDispatchNode(Array=RigVMVariableNode() {variable=control fk array}.Value) {template="DISPATCH_RigVMDispatch_ArrayClone(in Array,out Clone)"}
   n3 = RigVMDispatchNode(Index=0, Array=@Clone.Clone, ExecuteContext=@n2.B) {template="DISPATCH_RigVMDispatch_ArrayRemove(io Array,in Index)"}
   > RigVMVariableNode(Value=@n3.Array) {variable=animated ctrl array}
   > Reset Running Parent_1 = RigVMFunctionReferenceNode()
   > RigVMVariableNode() {variable=second half fk spawn}
   > &RigUnit_HierarchyAddControlTransform(
    InitialValue=(Rotation=(X=0,Y=0,Z=0,W=1),Translation=(X=0,Y=0,Z=0),Scale3D=(X=1,Y=1,Z=1))
    Name=tentacle_08_dup_offset
    OffsetSpace=GlobalSpace
    Parent=(Type=Control,Name="tentacle_07_fk")
    Settings="(InitialSpace=LocalSpace,bUsePreferredRotationOrder=False,PreferredRotationOrder=YZX,Limits=(LimitTranslationX=(bMinimum=False,bMaximum=False),LimitTranslationY=(bMinimum=False,bMaximum=False),LimitTranslationZ=(bMinimum=False,bMaximum=False),LimitPitch=(b ...(truncated)"
    OffsetTransform=&RigUnit_GetTransform(Item=(Type=Bone,Name="tentacle_08_dup"), bInitial=False).Transform
   ) {template="SpawnControl::Execute(in InitialValue,in Settings,in OffsetTransform,in OffsetSpace,in Parent,in Name,out Item)"}
   For_Each_2 = RigVMDispatchNode(Array=RigVMVariableNode() {variable=offset tentacle bones}.Value, ExecuteContext=@n2.A) {template="DISPATCH_RigVMDispatch_ArrayIterator(in Array,out Element,out Index,out Count,out Ratio)"}
   n4 = RigVMDispatchNode(
    Index=&RigVMFunction_MathIntAdd(B=5, A=@For_Each_2.Index) {template="Add::Execute(in A,in B,out Result)"}.Result
    Array=RigVMVariableNode() {variable=fk control scale array}.Value
   ) {template="DISPATCH_RigVMDispatch_ArrayGetAtIndex(in Array,in Index,out Element)"}
   n5 = RigVMDispatchNode(B=0, A=@For_Each_2.Index) {template="DISPATCH_RigVMDispatch_CoreEquals(in A,in B,out Result)"}
   > n6 = &RigUnit_HierarchyAddControlTransform(
    InitialValue=(Rotation=(X=0,Y=0,Z=0,W=1),Translation=(X=0,Y=0,Z=0),Scale3D=(X=1,Y=1,Z=1))
    OffsetSpace=GlobalSpace
    Settings="(InitialSpace=LocalSpace,bUsePreferredRotationOrder=true,PreferredRotationOrder=YZX,Limits=(LimitTranslationX=(bMinimum=False,bMaximum=False),LimitTranslationY=(bMinimum=False,bMaximum=False),LimitTranslationZ=(bMinimum=False,bMaximum=False),LimitPitch=(bM ...(truncated)"
    Parent=RigVMDispatchNode(
     True=(Type=Control,Name="tentacle_08_offset")
     Condition=&RigVMFunction_Contains(Search=07_dup, Name=@For_Each_2.Element.Name) {template="Contains::Execute(in Name,in Search,out Result)"}.Result
     False=RigVMVariableNode() {variable=running parent}.Value
    ) {template="DISPATCH_RigVMDispatch_If(in Condition,in True,in False,out Result)"}.Result
    Settings.Shape.Transform.Scale3D=RigVMDispatchNode(
     False=/Script/RigVM.RigVMFunction_MathVectorMul(B=(X=1.7,Y=1.7,Z=1.7), A=@n4.Element) {is=struct RigVMUnitNode, template="Multiply::Execute(in A,in B,out Result)"}.Result
     True=&RigVMFunction_MathVectorMul(B=(X=2.5,Y=2.5,Z=2.5), A=@n4.Element) {template="Multiply::Execute(in A,in B,out Result)"}.Result
     Condition=@n5.Result
    ) {template="DISPATCH_RigVMDispatch_If(in Condition,in True,in False,out Result)"}.Result
    Settings.PreferredRotationOrder=RigVMDispatchNode(True=YZX, Condition=@n5.Result) {template="DISPATCH_RigVMDispatch_If(in Condition,in True,in False,out Result)"}.Result
    Settings.Shape.Color=RigVMDispatchNode(False=(R=0,G=1,B=0,A=1), True=(R=0,G=1,B=0.2,A=1), Condition=@n5.Result) {template="DISPATCH_RigVMDispatch_If(in Condition,in True,in False,out Result)"}.Result
    OffsetTransform=&RigUnit_GetTransform(bInitial=False, Item=@For_Each_2.Element).Transform
    Name=&RigVMFunction_NameConcat(B=_offset, A=@For_Each_2.Element.Name) {template="Concat::Execute(in A,in B,out Result)"}.Result
   ) {template="SpawnControl::Execute(in InitialValue,in Settings,in OffsetTransform,in OffsetSpace,in Parent,in Name,out Item)"}
   > RigVMDispatchNode(Array=RigVMVariableNode() {variable=offset tentacle controls}.Value, Element=@n6.Item) {template="DISPATCH_RigVMDispatch_ArrayAdd(io Array,in Element,out Index)"}
   > RigVMVariableNode(Value=@n6.Item) {variable=running parent}
   n7 = RigVMDispatchNode(Index=0, Array=RigVMVariableNode() {variable=skeleton array}.Value) {template="DISPATCH_RigVMDispatch_ArrayGetAtIndex(in Array,in Index,out Element)"}
   n8 = &RigUnit_HierarchyAddControlTransform(
    InitialValue=(Rotation=(X=0,Y=0,Z=0,W=1),Translation=(X=0,Y=0,Z=0),Scale3D=(X=1,Y=1,Z=1))
    OffsetSpace=GlobalSpace
    Settings="(InitialSpace=LocalSpace,bUsePreferredRotationOrder=true,PreferredRotationOrder=YZX,Limits=(LimitTranslationX=(bMinimum=False,bMaximum=False),LimitTranslationY=(bMinimum=False,bMaximum=False),LimitTranslationZ=(bMinimum=False,bMaximum=False),LimitPitch=(bM ...(truncated)"
    OffsetTransform=&RigUnit_GetTransform(bInitial=False, Item=@n7.Element).Transform
    Parent=RigVMVariableNode() {variable=running parent}.Value
    ExecutePin=@RigVMFunction_Sequence_2.A
    Name=@n7.Element.Name
   ) {template="SpawnControl::Execute(in InitialValue,in Settings,in OffsetTransform,in OffsetSpace,in Parent,in Name,out Item)"}
   > RigVMVariableNode(Value=@n8.Item) {variable=running parent}
   > RigVMDispatchNode(Array=RigVMVariableNode() {variable=control fk array}.Value, Element=@n8.Item) {template="DISPATCH_RigVMDispatch_ArrayAdd(io Array,in Element,out Index)"}
   n9 = &RigVMFunction_Sequence(ExecuteContext=@RigVMFunction_Sequence_2.B)
   RigVMVariableNode(Value=(X=150,Y=150,Z=150), ExecuteContext=@n9.B) {variable=control fk scale}
   > Reset Running Parent = RigVMFunctionReferenceNode()
   For_Each = RigVMDispatchNode(Array=RigVMVariableNode() {variable=tentacle bones}.Value, ExecuteContext=@n9.A) {template="DISPATCH_RigVMDispatch_ArrayIterator(in Array,out Element,out Index,out Count,out Ratio)"}
   > RigVMVariableNode(Value=&RigVMFunction_MathIntAdd(B=1, A=@For_Each.Index) {template="Add::Execute(in A,in B,out Result)"}.Result) {variable=bone count}
   n10 = &RigVMFunction_MathDoubleSub(A=2500, B=1000)
   RigVMFunction_MathIntToDouble_7 = /Script/RigVM.RigVMFunction_MathIntToFloat(Value=@For_Each.Index) {is=struct RigVMUnitNode, template="Cast::Execute(in Value,out Result)"}
   RigVMFunction_MathIntToFloat = &RigVMFunction_MathIntToFloat(Value=@For_Each.Count) {template="Cast::Execute(in Value,out Result)"}
   n11 = &RigVMFunction_MathDoubleDiv(A=@RigVMFunction_MathIntToDouble_7.Result, B=@RigVMFunction_MathIntToFloat.Result)
   Concat = &RigVMFunction_NameConcat(B=_fk, A=@For_Each.Element.Name) {template="Concat::Execute(in A,in B,out Result)"}
   > SpawnControl = &RigUnit_HierarchyAddControlTransform(
    InitialValue=(Rotation=(X=0,Y=0,Z=0,W=1),Translation=(X=0,Y=0,Z=0),Scale3D=(X=1,Y=1,Z=1))
    OffsetSpace=GlobalSpace
    Settings="(InitialSpace=LocalSpace,bUsePreferredRotationOrder=true,PreferredRotationOrder=YXZ,Limits=(LimitTranslationX=(bMinimum=False,bMaximum=False),LimitTranslationY=(bMinimum=False,bMaximum=False),LimitTranslationZ=(bMinimum=False,bMaximum=False),LimitPitch=(bM ...(truncated)"
    Parent=RigVMVariableNode() {variable=running parent}.Value
    Settings.Shape.Transform.Scale3D=RigVMVariableNode() {variable=control fk scale}.Value
    Settings.Shape.Transform.Translation.Z=&RigVMFunction_MathDoubleSub(A=2500, B=&RigVMFunction_MathDoubleMul(B=@n10.Result, A=@n11.Result).Result).Result
    OffsetTransform=&RigUnit_GetTransform(bInitial=False, Item=@For_Each.Element).Transform
    Name=@Concat.Result
   ) {template="SpawnControl::Execute(in InitialValue,in Settings,in OffsetTransform,in OffsetSpace,in Parent,in Name,out Item)"}
   > RigVMVariableNode(Value=@SpawnControl.Item) {variable=running parent}
   > Add = RigVMDispatchNode(Array=RigVMVariableNode() {variable=control fk array}.Value, Element=@SpawnControl.Item) {template="DISPATCH_RigVMDispatch_ArrayAdd(io Array,in Element,out Index)"}
   RigVMFunction_MathIntToDouble = &RigVMFunction_MathIntToFloat(Value=&RigVMFunction_MathIntAdd(B=1, A=@For_Each.Index) {template="Add::Execute(in A,in B,out Result)"}.Result) {template="Cast::Execute(in Value,out Result)"}
   Num = RigVMDispatchNode(Array=RigVMVariableNode() {variable=tentacle bones}.Value) {template="DISPATCH_RigVMDispatch_ArrayGetNum(in Array,out Num)"}
   RigVMFunction_MathIntToDouble_1 = &RigVMFunction_MathIntToFloat(Value=&RigVMFunction_MathIntAdd(B=1, A=@Num.Num) {template="Add::Execute(in A,in B,out Result)"}.Result) {template="Cast::Execute(in Value,out Result)"}
   Divide = &RigVMFunction_MathDoubleDiv(A=@RigVMFunction_MathIntToDouble.Result, B=@RigVMFunction_MathIntToDouble_1.Result)
   Subtract = &RigVMFunction_MathDoubleSub(
    A=150
    B=&RigVMFunction_MathDoubleMul(A=@Divide.Result, B=&RigVMFunction_MathDoubleSub(A=150, B=100).Result).Result
   )
   > RigVMVariableNode(Value=(X=0,Y=0,Z=500), Value.X=@Subtract.Result, Value.Y=@Subtract.Result, Value.Z=@Subtract.Result) {variable=control fk scale}
   > n12 = RigVMDispatchNode(Element=RigVMVariableNode() {variable=control fk scale}.Value, Array=RigVMVariableNode() {variable=fk control scale array}.Value) {template="DISPATCH_RigVMDispatch_ArrayAdd(io Array,in Element,out Index)"}
   n13 = &RigVMFunction_ControlFlowBranch(
    ExecuteContext=@n12.ExecuteContext
    Condition=&RigVMFunction_Contains(Search=08, Name=@Concat.Result) {template="Contains::Execute(in Name,in Search,out Result)"}.Result
   )
   n14 = &RigVMFunction_ControlFlowBranch(ExecuteContext=@n13.Completed, Condition=RigVMVariableNode() {variable=second half fk spawn}.Value)
   RigVMDispatchNode(ExecuteContext=@n14.True, Array=RigVMVariableNode() {variable=upper tentacle controls}.Value, Element=@SpawnControl.Item) {template="DISPATCH_RigVMDispatch_ArrayAdd(io Array,in Element,out Index)"}
   RigVMVariableNode(Value=true, ExecuteContext=@n13.True) {variable=second half fk spawn}
  ]}
 New Function_ContainedGraph {nodes [
   RigVMFunctionEntryNode()
   > RigVMFunction_Sequence = RigVMAggregateNode()
   RigVMFunctionReturnNode(ExecuteContext=@RigVMFunction_Sequence.D)
   RigVMFunction_ForLoopCount = /Script/RigVM.RigVMFunction_ForLoopCount(Count=RigVMVariableNode() {variable=spline points}.Value, ExecutePin=@RigVMFunction_Sequence.B) {is=struct RigVMUnitNode}
   n1 = RigVMDispatchNode(Array=RigVMVariableNode() {variable=spline transforms}.Value, Index=@RigVMFunction_ForLoopCount.Index) {template="DISPATCH_RigVMDispatch_ArrayGetAtIndex(in Array,in Index,out Element)"}
   Int to Name_1 = /Script/RigVM.RigVMFunction_MathIntToName(
    PaddedSize=2
    Number=&RigVMFunction_MathIntAdd(B=1, A=@RigVMFunction_ForLoopCount.Index) {template="Add::Execute(in A,in B,out Result)"}.Result
   ) {is=struct RigVMUnitNode, template="Int to Name::Execute(in Number,in PaddedSize,out Result)"}
   > n2 = &RigUnit_HierarchyAddNull(
    Space=GlobalSpace
    Transform=(Rotation=(X=0.5,Y=-0.5,Z=0.5,W=0.5),Translation=(X=0,Y=0,Z=0),Scale3D=(X=1,Y=1,Z=1))
    Transform.Rotation=@n1.Element.Rotation
    Transform.Translation=@n1.Element.Translation
    Name=&RigVMFunction_NameConcat(A=spline_, B=@"Int to Name_1".Result) {template="Concat::Execute(in A,in B,out Result)"}.Result
    Parent=RigVMVariableNode() {variable=running parent}.Value
   )
   n3 = RigVMVariableNode() {variable=spline control scale}
   > n4 = &RigUnit_HierarchyAddControlTransform(
    InitialValue=(Rotation=(X=0,Y=0,Z=0,W=1),Translation=(X=0,Y=0,Z=0),Scale3D=(X=1,Y=1,Z=1))
    OffsetSpace=LocalSpace
    OffsetTransform=(Rotation=(X=0,Y=0,Z=0,W=1),Translation=(X=0,Y=0,Z=0),Scale3D=(X=1,Y=1,Z=1))
    Settings="(InitialSpace=LocalSpace,bUsePreferredRotationOrder=False,PreferredRotationOrder=YZX,Limits=(LimitTranslationX=(bMinimum=False,bMaximum=False),LimitTranslationY=(bMinimum=False,bMaximum=False),LimitTranslationZ=(bMinimum=False,bMaximum=False),LimitPitch=(b ...(truncated)"
    Parent=@n2.Item
    Name=@n2.Item.Name
    Settings.Shape.Transform.Scale3D=@n3.Value
   ) {template="SpawnControl::Execute(in InitialValue,in Settings,in OffsetTransform,in OffsetSpace,in Parent,in Name,out Item)"}
   > RigVMVariableNode(Value=@n4.Item) {variable=running parent}
   > RigVMDispatchNode(Array=RigVMVariableNode() {variable=spline ctrl array}.Value, Element=@n4.Item) {template="DISPATCH_RigVMDispatch_ArrayAdd(io Array,in Element,out Index)"}
   Multiply = &RigVMFunction_MathVectorMul(B=(X=0.8,Y=0.8,Z=0.8), A=@n3.Value) {template="Multiply::Execute(in A,in B,out Result)"}
   > n5 = &RigUnit_HierarchyAddControlTransform(
    InitialValue=(Rotation=(X=0,Y=0,Z=0,W=1),Translation=(X=0,Y=0,Z=0),Scale3D=(X=1,Y=1,Z=1))
    OffsetSpace=LocalSpace
    OffsetTransform=(Rotation=(X=0,Y=0,Z=0,W=1),Translation=(X=0,Y=0,Z=0),Scale3D=(X=1,Y=1,Z=1))
    Settings="(InitialSpace=LocalSpace,bUsePreferredRotationOrder=False,PreferredRotationOrder=YZX,Limits=(LimitTranslationX=(bMinimum=False,bMaximum=False),LimitTranslationY=(bMinimum=False,bMaximum=False),LimitTranslationZ=(bMinimum=False,bMaximum=False),LimitPitch=(b ...(truncated)"
    Name=&RigVMFunction_NameConcat(B=_direct, A=@n4.Item.Name) {template="Concat::Execute(in A,in B,out Result)"}.Result
    Settings.Shape.Transform.Scale3D=@Multiply.Result
    Parent=@n4.Item
   ) {template="SpawnControl::Execute(in InitialValue,in Settings,in OffsetTransform,in OffsetSpace,in Parent,in Name,out Item)"}
   > RigVMDispatchNode(Array=RigVMVariableNode() {variable=spline direct ctrl array}.Value, Element=@n5.Item) {template="DISPATCH_RigVMDispatch_ArrayAdd(io Array,in Element,out Index)"}
   RigVMFunction_MathIntToDouble_5 = &RigVMFunction_MathIntToFloat(Value=RigVMVariableNode() {variable=spline points}.Value) {template="Cast::Execute(in Value,out Result)"}
   RigVMFunction_MathIntToDouble_2 = &RigVMFunction_MathIntToFloat(Value=@RigVMFunction_ForLoopCount.Index) {template="Cast::Execute(in Value,out Result)"}
   n6 = &RigVMFunction_MathDoubleSub(
    A=780
    B=&RigVMFunction_MathDoubleMul(
     A=&RigVMFunction_MathDoubleDiv(B=@RigVMFunction_MathIntToDouble_5.Result, A=@RigVMFunction_MathIntToDouble_2.Result).Result
     B=&RigVMFunction_MathDoubleSub(A=780, B=300).Result
    ).Result
   )
   > RigVMVariableNode(Value=(X=0,Y=0,Z=700), Value.X=@n6.Result, Value.Y=@n6.Result) {variable=spline control scale}
   RigVMVariableNode(Value=(X=900,Y=900,Z=900), ExecuteContext=@RigVMFunction_Sequence.C) {variable=spline control scale}
   > Reset Running Parent = RigVMFunctionReferenceNode()
   n7 = &RigVMFunction_ForLoopCount(Count=RigVMVariableNode() {variable=spline points}.Value, ExecutePin=@RigVMFunction_Sequence.A)
   RigVMFunction_MathIntToDouble_8 = &RigVMFunction_MathIntToFloat(Value=/Script/RigVM.RigVMFunction_MathIntSub(B=1, A=RigVMVariableNode() {variable=spline points}.Value) {is=struct RigVMUnitNode, template="Subtract::Execute(in A,in B,out Result)"}.Result) {template="Cast::Execute(in Value,out Result)"}
   Exponential = /Script/RigVM.RigVMFunction_MathDoubleExponential(Value=&RigVMFunction_MathDoubleMul(A=-0.2, B=@RigVMFunction_MathIntToDouble_8.Result).Result) {is=struct RigVMUnitNode, template="Exponential::Execute(in Value,out Result)"}
   n8 = &RigVMFunction_MathDoubleSub(A=1, B=@Exponential.Result)
   RigVMFunction_MathIntToDouble_9 = &RigVMFunction_MathIntToFloat(Value=@n7.Index) {template="Cast::Execute(in Value,out Result)"}
   n9 = &RigVMFunction_MathDoubleSub(
    A=1
    B=&RigVMFunction_MathDoubleExponential(Value=&RigVMFunction_MathDoubleMul(A=-0.2, B=@RigVMFunction_MathIntToDouble_9.Result).Result) {template="Exponential::Execute(in Value,out Result)"}.Result
   )
   > RigVMDispatchNode(
    Element=&RigUnit_ChainInfo(
     DebugScale=1000
     bCalculateStretch=True
     bDebug=false
     bInitial=false
     Param=&RigVMFunction_MathDoubleDiv(B=@n8.Result, A=@n9.Result).Result
     Items=RigVMVariableNode() {variable=tentacle bones}.Value
    ).InterpolatedTransform
    Array=RigVMVariableNode() {variable=spline transforms}.Value
   ) {template="DISPATCH_RigVMDispatch_ArrayAdd(io Array,in Element,out Index)"}
   n10 = RigVMVariableNode() {variable=tentacle bones}
   n11 = RigVMDispatchNode(Index=0, Array=@n10.Value) {template="DISPATCH_RigVMDispatch_ArrayGetAtIndex(in Array,in Index,out Element)"}
   n12 = RigVMDispatchNode(Index=-1, Array=@n10.Value) {template="DISPATCH_RigVMDispatch_ArrayGetAtIndex(in Array,in Index,out Element)"}
   RigVMVariableNode(
    Value=&RigVMFunction_MathVectorDistance(
     A=&RigUnit_GetTransform(bInitial=False, Item=@n11.Element).Transform.Translation
     B=&RigUnit_GetTransform(bInitial=False, Item=@n12.Element).Transform.Translation
    ).Result
    ExecuteContext=@n7.Completed
   ) {variable=spline original length}
   n13 = RigVMVariableNode() {variable=tentacle bones}
   > For_Each_5 = RigVMDispatchNode(Array=@n13.Value) {template="DISPATCH_RigVMDispatch_ArrayIterator(in Array,out Element,out Index,out Count,out Ratio)"}
   RigVMFunction_MathVectorDistance = &RigVMFunction_MathVectorDistance(
    A=&RigUnit_GetTransform(
     bInitial=False
     Item=RigVMDispatchNode(Index=0, Array=@n13.Value) {template="DISPATCH_RigVMDispatch_ArrayGetAtIndex(in Array,in Index,out Element)"}.Element
    ).Transform.Translation
    B=&RigUnit_GetTransform(bInitial=False, Item=@For_Each_5.Element).Transform.Translation
   )
   > RigVMDispatchNode(
    Element=&RigVMFunction_MathDoubleDiv(A=@RigVMFunction_MathVectorDistance.Result, B=RigVMVariableNode() {variable=spline original length}.Value).Result
    Array=RigVMVariableNode() {variable=bone chain percentage}.Value
   ) {template="DISPATCH_RigVMDispatch_ArrayAdd(io Array,in Element,out Index)"}
  ]}
 New Function_ContainedGraph {nodes [
   n1 = RigVMFunctionEntryNode()
   RigVMFunction_Sequence = &RigVMFunction_Sequence(ExecuteContext=@n1.ExecuteContext)
   RigVMFunctionReturnNode(ExecuteContext=@RigVMFunction_Sequence.B)
   RigVMFunction_ForLoopCount = &RigVMFunction_ForLoopCount(Count=5, ExecutePin=@RigVMFunction_Sequence.A)
   n2 = &RigUnit_ChainInfo(
    DebugScale=300
    Items=((Type=None,Name="None"),(Type=None,Name="None"),(Type=None,Name="None"))
    bCalculateStretch=True
    bDebug=true
    bInitial=False
    Items.0=RigVMDispatchNode(Index=@RigVMFunction_ForLoopCount.Index, Array=RigVMVariableNode() {variable=eye interp start bones}.Value) {template="DISPATCH_RigVMDispatch_ArrayGetAtIndex(in Array,in Index,out Element)"}.Element
    Items.2=RigVMDispatchNode(Index=@RigVMFunction_ForLoopCount.Index, Array=RigVMVariableNode() {variable=eye interp end bones}.Value) {template="DISPATCH_RigVMDispatch_ArrayGetAtIndex(in Array,in Index,out Element)"}.Element
    Param=RigVMDispatchNode(Index=@RigVMFunction_ForLoopCount.Index, Array=RigVMVariableNode() {variable=eye interp param}.Value) {template="DISPATCH_RigVMDispatch_ArrayGetAtIndex(in Array,in Index,out Element)"}.Element
    Items.1=RigVMDispatchNode(Index=@RigVMFunction_ForLoopCount.Index, Array=RigVMVariableNode() {variable=eye interp mid bones}.Value) {template="DISPATCH_RigVMDispatch_ArrayGetAtIndex(in Array,in Index,out Element)"}.Element
   )
   > RigVMFunction_DebugTransformMutableNoSpace = /Script/RigVM.RigVMFunction_DebugTransformMutableNoSpace(
    Color=(R=1,G=0,B=0,A=1)
    DebugDrawSettings=(DepthPriority=SDPG_Foreground,Lifetime=-1)
    Mode=Axes
    Scale=10
    Thickness=0
    WorldOffset=(Rotation=(X=0,Y=0,Z=0,W=1),Translation=(X=0,Y=0,Z=0),Scale3D=(X=1,Y=1,Z=1))
    bEnabled=True
    Transform=@n2.InterpolatedTransform
   ) {is=struct RigVMUnitNode}
  ]}
 New Function_ContainedGraph {nodes [
   n1 = RigVMFunctionEntryNode()
   RigVMFunction_Sequence = &RigVMFunction_Sequence(ExecuteContext=@n1.ExecuteContext)
   RigVMFunctionReturnNode(ExecuteContext=@RigVMFunction_Sequence.B)
   RigVMFunction_Sequence_5 = RigVMAggregateNode(ExecuteContext=@RigVMFunction_Sequence.A)
   n2 = &RigUnit_HierarchyImportFromSkeleton(NameSpace=None, bIncludeCurves=False, bIncludeMeshSockets=False, bIncludeVirtualBones=True, ExecutePin=@RigVMFunction_Sequence_5.A)
   > RigVMVariableNode(Value=@n2.Items) {variable=skeleton array}
   > RigVMVariableNode(Value=&RigUnit_CollectionChildrenArray(
     TypeToSearch=Bone
     bDefaultChildren=True
     bIncludeParent=False
     bRecursive=true
     Parent=RigVMDispatchNode(Index=0, Array=@n2.Items) {template="DISPATCH_RigVMDispatch_ArrayGetAtIndex(in Array,in Index,out Element)"}.Element
    ).Items) {variable=animated joints array}
   > For_Each_1 = RigVMDispatchNode(Array=RigVMVariableNode() {variable=skeleton array}.Value) {template="DISPATCH_RigVMDispatch_ArrayIterator(in Array,out Element,out Index,out Count,out Ratio)"}
   Contains = &RigVMFunction_Contains(Search=tentacle, Name=@For_Each_1.Element.Name) {template="Contains::Execute(in Name,in Search,out Result)"}
   RigVMFunction_MathBoolAnd_1 = RigVMAggregateNode(
    B=@Contains.Result
    A=&RigVMFunction_MathBoolNot(Value=&RigVMFunction_Contains(Search=dup, Name=@For_Each_1.Element.Name) {template="Contains::Execute(in Name,in Search,out Result)"}.Result).Result
    C=&RigVMFunction_MathBoolNot(Value=&RigVMFunction_Contains(Search=_offset, Name=@For_Each_1.Element.Name) {template="Contains::Execute(in Name,in Search,out Result)"}.Result).Result
   )
   RigVMFunction_ControlFlowBranch = &RigVMFunction_ControlFlowBranch(Condition=@RigVMFunction_MathBoolAnd_1.Result, ExecuteContext=@For_Each_1.ExecuteContext)
   RigVMDispatchNode(ExecuteContext=@RigVMFunction_ControlFlowBranch.True, Array=RigVMVariableNode() {variable=tentacle bones}.Value, Element=@For_Each_1.Element) {template="DISPATCH_RigVMDispatch_ArrayAdd(io Array,in Element,out Index)"}
   RigVMVariableNode(ExecuteContext=@RigVMFunction_Sequence_5.B) {variable=eye cycle count}
   > For_Each_3 = RigVMDispatchNode(Array=RigVMVariableNode() {variable=skeleton array}.Value) {template="DISPATCH_RigVMDispatch_ArrayIterator(in Array,out Element,out Index,out Count,out Ratio)"}
   n3 = RigVMVariableNode() {variable=eye names}
   n4 = RigVMVariableNode() {variable=eye cycle count}
   n5 = &RigVMFunction_ControlFlowBranch(
    Condition=&RigVMFunction_Contains(
     Search=RigVMDispatchNode(Array=@n3.Value, Index=@n4.Value) {template="DISPATCH_RigVMDispatch_ArrayGetAtIndex(in Array,in Index,out Element)"}.Element
     Name=@For_Each_3.Element.Name
    ) {template="Contains::Execute(in Name,in Search,out Result)"}.Result
    ExecuteContext=@For_Each_3.ExecuteContext
   )
   n6 = &RigUnit_CollectionChildrenArray(TypeToSearch=Bone, bDefaultChildren=True, bIncludeParent=true, bRecursive=true, Parent=@For_Each_3.Element)
   RigVMDispatchNode(Other=@n6.Items, ExecuteContext=@n5.True, Array=RigVMVariableNode() {variable=eye master array}.Value) {template="DISPATCH_RigVMDispatch_ArrayAppend(io Array,in Other)"}
   > n7 = RigVMDispatchNode(Index=RigVMVariableNode() {variable=eye cycle count}.Value) {template="DISPATCH_RigVMDispatch_SwitchInt32(in Index)"}
   RigVMVariableNode(ExecuteContext=@n7.Cases.0, Value=@n6.Items) {variable=eyeE}
   > RigVMVariableNode(Value=1) {variable=eye cycle count}
   RigVMVariableNode(ExecuteContext=@n7.Cases.1, Value=@n6.Items) {variable=eyeD}
   > RigVMVariableNode(Value=2) {variable=eye cycle count}
   RigVMVariableNode(ExecuteContext=@n7.Cases.2, Value=@n6.Items) {variable=eyeC}
   > RigVMVariableNode(Value=3) {variable=eye cycle count}
   RigVMVariableNode(ExecuteContext=@n7.Cases.3, Value=@n6.Items) {variable=eyeB}
   > RigVMVariableNode(Value=4) {variable=eye cycle count}
   RigVMVariableNode(ExecuteContext=@n7.Cases.4, Value=@n6.Items) {variable=eyeA}
   > RigVMVariableNode(Value=0) {variable=eye cycle count}
   Make_Array_3 = RigVMDispatchNode(Values="((Type=Bone,Name=\"tentacle_08\"),((),Type=Bone,Name=\"tentacle_09\"),((),Type=Bone,Name=\"tentacle_10\"),((),Type=Bone,Name=\"tentacle_11\"),((),Type=Bone,Name=\"tentacle_12\"),((),Type=Bone,Name=\"tentacle_13\"),((),Type=Bone,Name=\"tentacle_14\"),((),Type=Bone,Name=\" ...(truncated)") {template="DISPATCH_RigVMDispatch_ArrayMake(in Values,out Array)"}
   RigVMVariableNode(Value=@Make_Array_3.Array, ExecuteContext=@RigVMFunction_Sequence_5.C) {variable=offset tentacle bones}
   Make_Array_4 = RigVMDispatchNode(Values=((Type=Bone,Name="tentacle_07_dup"),(Type=Bone,Name="tentacle_06_dup"))) {template="DISPATCH_RigVMDispatch_ArrayMake(in Values,out Array)"}
   > n8 = RigVMDispatchNode(
    Array=RigVMDispatchNode(Array=RigVMVariableNode() {variable=tentacle bones}.Value) {template="DISPATCH_RigVMDispatch_ArrayClone(in Array,out Clone)"}.Clone
    Other=@Make_Array_4.Array
   ) {template="DISPATCH_RigVMDispatch_ArrayAppend(io Array,in Other)"}
   > RigVMVariableNode(Value=RigVMDispatchNode(Array=@n8.Array, Other=RigVMVariableNode() {variable=offset tentacle bones}.Value) {template="DISPATCH_RigVMDispatch_ArrayDifference(in Array,in Other,out Result)"}.Result) {variable=non offset tentacle bones}
   > n9 = RigVMDispatchNode(
    Index=-1
    Array=RigVMDispatchNode(Array=RigVMVariableNode() {variable=offset tentacle bones}.Value) {template="DISPATCH_RigVMDispatch_ArrayClone(in Array,out Clone)"}.Clone
   ) {template="DISPATCH_RigVMDispatch_ArrayRemove(io Array,in Index)"}
   > n10 = RigVMDispatchNode(Index=-1, Array=@n9.Array) {template="DISPATCH_RigVMDispatch_ArrayRemove(io Array,in Index)"}
   > RigVMVariableNode(Value=@n10.Array) {variable=offset tentacle bones exclude dupes}
  ]}
 New Function_ContainedGraph {nodes [
   n1 = RigVMFunctionEntryNode()
   RigVMFunction_Sequence = &RigVMFunction_Sequence(ExecuteContext=@n1.ExecuteContext)
   RigVMFunctionReturnNode(ExecuteContext=@RigVMFunction_Sequence.B)
   n2 = &RigVMFunction_Sequence(ExecuteContext=@RigVMFunction_Sequence.A)
   n3 = &RigVMFunction_Sequence(ExecuteContext=@n2.B)
   n4 = &RigVMFunction_Sequence(ExecuteContext=@n3.B)
   n5 = &RigUnit_HierarchyAddAnimationChannelFloat(
    InitialValue=1
    LimitsEnabled=(Enabled=(bMinimum=True,bMaximum=True))
    MaximumValue=1
    MinimumValue=0
    Name=Cut Constraints - Rotational
    Parent=(Type=Control,Name="control rig options")
    ExecutePin=@n4.A
   ) {template="SpawnAnimationChannel::Execute(in InitialValue,in MinimumValue,in MaximumValue,in LimitsEnabled,in Parent,in Name,out Item)"}
   > RigVMVariableNode(Value=@n5.Item) {variable=cut constraints rotational channel}
   > n6 = &RigUnit_HierarchyAddAnimationChannelFloat(
    InitialValue=1
    LimitsEnabled=(Enabled=(bMinimum=True,bMaximum=True))
    MaximumValue=1
    MinimumValue=0
    Name=Cut Constraints - Positional
    Parent=(Type=Control,Name="control rig options")
   ) {template="SpawnAnimationChannel::Execute(in InitialValue,in MinimumValue,in MaximumValue,in LimitsEnabled,in Parent,in Name,out Item)"}
   > RigVMVariableNode(Value=@n6.Item) {variable=cut constraints positional channel}
   > n7 = /Script/ControlRig.RigUnit_HierarchyAddAnimationChannelBool(
    InitialValue=false
    LimitsEnabled=()
    MaximumValue=True
    MinimumValue=False
    Name=Enable Cut Offset
    Parent=(Type=Control,Name="control rig options")
   ) {
    is=struct RigVMUnitNode
    template="SpawnAnimationChannel::Execute(in InitialValue,in MinimumValue,in MaximumValue,in LimitsEnabled,in Parent,in Name,out Item)"
   }
   Make_Array_7 = RigVMDispatchNode(
    Values=((Type=None,Name="None"),((),Type=None,Name="None"),((),Type=None,Name="None"))
    Values.0=@n5.Item
    Values.1=@n6.Item
    Values.2=@n7.Item
   ) {template="DISPATCH_RigVMDispatch_ArrayMake(in Values,out Array)"}
   > RigVMDispatchNode(Other=@Make_Array_7.Array, Array=RigVMVariableNode() {variable=Channels array}.Value) {template="DISPATCH_RigVMDispatch_ArrayAppend(io Array,in Other)"}
   n8 = &RigUnit_HierarchyAddAnimationChannelBool(
    InitialValue=true
    LimitsEnabled=()
    MaximumValue=True
    MinimumValue=False
    Name=Offsets follow FK
    Parent=(Type=Control,Name="control rig options")
    ExecutePin=@n4.B
   ) {template="SpawnAnimationChannel::Execute(in InitialValue,in MinimumValue,in MaximumValue,in LimitsEnabled,in Parent,in Name,out Item)"}
   Make_Array_8 = RigVMDispatchNode(Values=((Type=None,Name="None")), Values.0=@n8.Item) {template="DISPATCH_RigVMDispatch_ArrayMake(in Values,out Array)"}
   > RigVMDispatchNode(Other=@Make_Array_8.Array, Array=RigVMVariableNode() {variable=Channels array}.Value) {template="DISPATCH_RigVMDispatch_ArrayAppend(io Array,in Other)"}
   n9 = &RigVMFunction_Sequence(ExecuteContext=@n3.A)
   n10 = &RigUnit_HierarchyAddAnimationChannelBool(
    InitialValue=false
    LimitsEnabled=()
    MaximumValue=True
    MinimumValue=False
    Name=Eye Control Visibility
    Parent=(Type=Control,Name="control rig options")
    ExecutePin=@n9.A
   ) {template="SpawnAnimationChannel::Execute(in InitialValue,in MinimumValue,in MaximumValue,in LimitsEnabled,in Parent,in Name,out Item)"}
   > n11 = &RigUnit_HierarchyAddAnimationChannelBool(
    InitialValue=true
    LimitsEnabled=()
    MaximumValue=True
    MinimumValue=False
    Name="Procedural Blink "
    Parent=(Type=Control,Name="control rig options")
   ) {template="SpawnAnimationChannel::Execute(in InitialValue,in MinimumValue,in MaximumValue,in LimitsEnabled,in Parent,in Name,out Item)"}
   > n12 = &RigUnit_HierarchyAddAnimationChannelFloat(
    InitialValue=0
    LimitsEnabled=(Enabled=(bMinimum=True,bMaximum=True))
    MaximumValue=1
    MinimumValue=0
    Name=Master Blink
    Parent=(Type=Control,Name="control rig options")
   ) {template="SpawnAnimationChannel::Execute(in InitialValue,in MinimumValue,in MaximumValue,in LimitsEnabled,in Parent,in Name,out Item)"}
   Make_Array_5 = RigVMDispatchNode(
    Values=((Type=None,Name="None"),((),Type=None,Name="None"),((),Type=None,Name="None"))
    Values.2=@n12.Item
    Values.1=@n11.Item
    Values.0=@n10.Item
   ) {template="DISPATCH_RigVMDispatch_ArrayMake(in Values,out Array)"}
   > RigVMDispatchNode(Other=@Make_Array_5.Array, Array=RigVMVariableNode() {variable=Channels array}.Value) {template="DISPATCH_RigVMDispatch_ArrayAppend(io Array,in Other)"}
   n13 = &RigUnit_HierarchyAddAnimationChannelFloat(
    InitialValue=0
    LimitsEnabled=(Enabled=(bMinimum=True,bMaximum=True))
    MaximumValue=1
    MinimumValue=0
    Name=Master Inner Blink
    Parent=(Type=Control,Name="control rig options")
    ExecutePin=@n9.B
   ) {template="SpawnAnimationChannel::Execute(in InitialValue,in MinimumValue,in MaximumValue,in LimitsEnabled,in Parent,in Name,out Item)"}
   > n14 = &RigUnit_HierarchyAddAnimationChannelFloat(
    InitialValue=0
    LimitsEnabled=(Enabled=(bMinimum=True,bMaximum=True))
    MaximumValue=0.3
    MinimumValue=-0.3
    Name=Eyes Look X
    Parent=(Type=Control,Name="control rig options")
   ) {template="SpawnAnimationChannel::Execute(in InitialValue,in MinimumValue,in MaximumValue,in LimitsEnabled,in Parent,in Name,out Item)"}
   > n15 = &RigUnit_HierarchyAddAnimationChannelFloat(
    InitialValue=0
    LimitsEnabled=(Enabled=(bMinimum=True,bMaximum=True))
    MaximumValue=0.3
    MinimumValue=-0.3
    Name=Eyes Look Y
    Parent=(Type=Control,Name="control rig options")
   ) {template="SpawnAnimationChannel::Execute(in InitialValue,in MinimumValue,in MaximumValue,in LimitsEnabled,in Parent,in Name,out Item)"}
   Make_Array_6 = RigVMDispatchNode(
    Values=((Type=None,Name="None"),((),Type=None,Name="None"),((),Type=None,Name="None"))
    Values.2=@n15.Item
    Values.1=@n14.Item
    Values.0=@n13.Item
   ) {template="DISPATCH_RigVMDispatch_ArrayMake(in Values,out Array)"}
   > RigVMDispatchNode(Other=@Make_Array_6.Array, Array=RigVMVariableNode() {variable=Channels array}.Value) {template="DISPATCH_RigVMDispatch_ArrayAppend(io Array,in Other)"}
   n16 = &RigVMFunction_Sequence(ExecuteContext=@n2.A)
   n17 = &RigUnit_HierarchyAddAnimationChannelBool(
    InitialValue=true
    LimitsEnabled=()
    MaximumValue=True
    MinimumValue=False
    Name=Root Visibility
    Parent=(Type=Control,Name="control rig options")
    ExecutePin=@n16.A
   ) {template="SpawnAnimationChannel::Execute(in InitialValue,in MinimumValue,in MaximumValue,in LimitsEnabled,in Parent,in Name,out Item)"}
   > n18 = &RigUnit_HierarchyAddAnimationChannelBool(
    InitialValue=true
    LimitsEnabled=()
    MaximumValue=True
    MinimumValue=False
    Name=FK Rig Visibility
    Parent=(Type=Control,Name="control rig options")
   ) {template="SpawnAnimationChannel::Execute(in InitialValue,in MinimumValue,in MaximumValue,in LimitsEnabled,in Parent,in Name,out Item)"}
   > n19 = &RigUnit_HierarchyAddAnimationChannelBool(
    InitialValue=true
    LimitsEnabled=()
    MaximumValue=True
    MinimumValue=False
    Name=Spline Rig Visibility
    Parent=(Type=Control,Name="control rig options")
   ) {template="SpawnAnimationChannel::Execute(in InitialValue,in MinimumValue,in MaximumValue,in LimitsEnabled,in Parent,in Name,out Item)"}
   Make_Array_3 = RigVMDispatchNode(
    Values=((Type=None,Name="None"),((),Type=None,Name="None"),((),Type=None,Name="None"))
    Values.0=@n17.Item
    Values.1=@n18.Item
    Values.2=@n19.Item
   ) {template="DISPATCH_RigVMDispatch_ArrayMake(in Values,out Array)"}
   > RigVMDispatchNode(Other=@Make_Array_3.Array, Array=RigVMVariableNode() {variable=Channels array}.Value) {template="DISPATCH_RigVMDispatch_ArrayAppend(io Array,in Other)"}
   n20 = &RigUnit_HierarchyAddAnimationChannelBool(
    InitialValue=false
    LimitsEnabled=()
    MaximumValue=True
    MinimumValue=False
    Name=Spline Draw Visibility
    Parent=(Type=Control,Name="control rig options")
    ExecutePin=@n16.B
   ) {template="SpawnAnimationChannel::Execute(in InitialValue,in MinimumValue,in MaximumValue,in LimitsEnabled,in Parent,in Name,out Item)"}
   > n21 = &RigUnit_HierarchyAddAnimationChannelFloat(
    InitialValue=0
    LimitsEnabled=(Enabled=(bMinimum=True,bMaximum=True))
    MaximumValue=1
    MinimumValue=0
    Name=Spline Stretch
    Parent=(Type=Control,Name="control rig options")
   ) {template="SpawnAnimationChannel::Execute(in InitialValue,in MinimumValue,in MaximumValue,in LimitsEnabled,in Parent,in Name,out Item)"}
   > RigVMVariableNode(Value=@n21.Item) {variable=Spline Stretch Channel}
   > n22 = &RigUnit_HierarchyAddAnimationChannelBool(
    InitialValue=False
    LimitsEnabled=()
    MaximumValue=True
    MinimumValue=False
    Name=Hermite Spline
    Parent=(Type=Control,Name="control rig options")
   ) {template="SpawnAnimationChannel::Execute(in InitialValue,in MinimumValue,in MaximumValue,in LimitsEnabled,in Parent,in Name,out Item)"}
   Make_Array_4 = RigVMDispatchNode(
    Values=((Type=None,Name="None"),((),Type=None,Name="None"),((),Type=None,Name="None"))
    Values.0=@n20.Item
    Values.1=@n21.Item
    Values.2=@n22.Item
   ) {template="DISPATCH_RigVMDispatch_ArrayMake(in Values,out Array)"}
   > RigVMDispatchNode(Other=@Make_Array_4.Array, Array=RigVMVariableNode() {variable=Channels array}.Value) {template="DISPATCH_RigVMDispatch_ArrayAppend(io Array,in Other)"}
  ]}
 New Function_ContainedGraph {nodes [
   n1 = RigVMFunctionEntryNode()
   RigVMFunction_Sequence = &RigVMFunction_Sequence(ExecuteContext=@n1.ExecuteContext)
   RigVMFunctionReturnNode(ExecuteContext=@RigVMFunction_Sequence.B)
   n2 = &RigVMFunction_Sequence(ExecuteContext=@RigVMFunction_Sequence.A)
   RigVMDispatchNode(ExecuteContext=@n2.B, Array=RigVMVariableNode() {variable=innerlid control array}.Value) {template="DISPATCH_RigVMDispatch_ArrayReset(io Array)"}
   > RigVMDispatchNode(Array=RigVMVariableNode() {variable=eye ball control array}.Value) {template="DISPATCH_RigVMDispatch_ArrayReset(io Array)"}
   > RigVMDispatchNode(Array=RigVMVariableNode() {variable=bone chain percentage}.Value) {template="DISPATCH_RigVMDispatch_ArrayReset(io Array)"}
   > RigVMDispatchNode(Array=RigVMVariableNode() {variable=Channels array}.Value) {template="DISPATCH_RigVMDispatch_ArrayReset(io Array)"}
   > RigVMDispatchNode(Array=RigVMVariableNode() {variable=upper tentacle controls}.Value) {template="DISPATCH_RigVMDispatch_ArrayReset(io Array)"}
   > RigVMDispatchNode(Array=RigVMVariableNode() {variable=offset tentacle bones}.Value) {template="DISPATCH_RigVMDispatch_ArrayReset(io Array)"}
   > RigVMDispatchNode(Array=RigVMVariableNode() {variable=offset tentacle controls}.Value) {template="DISPATCH_RigVMDispatch_ArrayReset(io Array)"}
   > RigVMDispatchNode(Array=RigVMVariableNode() {variable=fk control scale array}.Value) {template="DISPATCH_RigVMDispatch_ArrayReset(io Array)"}
   > RigVMDispatchNode(Array=RigVMVariableNode() {variable=non offset tentacle bones}.Value) {template="DISPATCH_RigVMDispatch_ArrayReset(io Array)"}
   > RigVMDispatchNode(Array=RigVMVariableNode() {variable=offset tentacle bones exclude dupes}.Value) {template="DISPATCH_RigVMDispatch_ArrayReset(io Array)"}
   > RigVMDispatchNode(Array=RigVMVariableNode() {variable=eye null array}.Value) {template="DISPATCH_RigVMDispatch_ArrayReset(io Array)"}
   > RigVMDispatchNode(Array=RigVMVariableNode() {variable=eye master controls}.Value) {template="DISPATCH_RigVMDispatch_ArrayReset(io Array)"}
   > RigVMDispatchNode(Array=RigVMVariableNode() {variable=eye orb controls}.Value) {template="DISPATCH_RigVMDispatch_ArrayReset(io Array)"}
   > RigVMDispatchNode(Array=RigVMVariableNode() {variable=eye orb array}.Value) {template="DISPATCH_RigVMDispatch_ArrayReset(io Array)"}
   RigVMDispatchNode(Array=RigVMVariableNode() {variable=skeleton array}.Value, ExecuteContext=@n2.A) {template="DISPATCH_RigVMDispatch_ArrayReset(io Array)"}
   > RigVMDispatchNode(Array=RigVMVariableNode() {variable=control fk array}.Value) {template="DISPATCH_RigVMDispatch_ArrayReset(io Array)"}
   > RigVMDispatchNode(Array=RigVMVariableNode() {variable=spline direct ctrl array}.Value) {template="DISPATCH_RigVMDispatch_ArrayReset(io Array)"}
   > RigVMDispatchNode(Array=RigVMVariableNode() {variable=spline translations}.Value) {template="DISPATCH_RigVMDispatch_ArrayReset(io Array)"}
   > RigVMDispatchNode(Array=RigVMVariableNode() {variable=animated joints array}.Value) {template="DISPATCH_RigVMDispatch_ArrayReset(io Array)"}
   > RigVMDispatchNode(Array=RigVMVariableNode() {variable=animated ctrl array}.Value) {template="DISPATCH_RigVMDispatch_ArrayReset(io Array)"}
   > RigVMDispatchNode(Array=RigVMVariableNode() {variable=spline ctrl array}.Value) {template="DISPATCH_RigVMDispatch_ArrayReset(io Array)"}
   > RigVMDispatchNode(Array=RigVMVariableNode() {variable=tentacle bones}.Value) {template="DISPATCH_RigVMDispatch_ArrayReset(io Array)"}
   > RigVMDispatchNode(Array=RigVMVariableNode() {variable=spline transforms}.Value) {template="DISPATCH_RigVMDispatch_ArrayReset(io Array)"}
   > RigVMDispatchNode(Array=RigVMVariableNode() {variable=eye master array}.Value) {template="DISPATCH_RigVMDispatch_ArrayReset(io Array)"}
   > RigVMDispatchNode(Array=RigVMVariableNode() {variable=eye control array}.Value) {template="DISPATCH_RigVMDispatch_ArrayReset(io Array)"}
   > RigVMDispatchNode(Array=RigVMVariableNode() {variable=eye matching skel array}.Value) {template="DISPATCH_RigVMDispatch_ArrayReset(io Array)"}
   > RigVMDispatchNode(Array=RigVMVariableNode() {variable=blink channels}.Value) {template="DISPATCH_RigVMDispatch_ArrayReset(io Array)"}
   > RigVMDispatchNode(Array=RigVMVariableNode() {variable=innerlid skel array}.Value) {template="DISPATCH_RigVMDispatch_ArrayReset(io Array)"}
   > RigVMDispatchNode(Array=RigVMVariableNode() {variable=lid control array}.Value) {template="DISPATCH_RigVMDispatch_ArrayReset(io Array)"}
   > RigVMDispatchNode(Array=RigVMVariableNode() {variable=innerlidA array}.Value) {template="DISPATCH_RigVMDispatch_ArrayReset(io Array)"}
  ]}
 New Function_ContainedGraph {nodes [
   n1 = RigVMFunctionEntryNode()
   RigVMFunction_Sequence = &RigVMFunction_Sequence(ExecuteContext=@n1.ExecuteContext)
   RigVMVariableNode(
    Value=RigVMDispatchNode(Index=0, Array=RigVMVariableNode() {variable=control fk array}.Value) {template="DISPATCH_RigVMDispatch_ArrayGetAtIndex(in Array,in Index,out Element)"}.Element
    ExecuteContext=@RigVMFunction_Sequence.A
   ) {variable=running parent}
   RigVMFunctionReturnNode(ExecuteContext=@RigVMFunction_Sequence.B)
  ]}
 New Function_ContainedGraph {nodes [
   n1 = RigVMFunctionEntryNode()
   RigVMFunction_Sequence = &RigVMFunction_Sequence(ExecuteContext=@n1.ExecuteContext)
   RigVMFunctionReturnNode(ExecuteContext=@RigVMFunction_Sequence.B)
   For_Each_7 = RigVMDispatchNode(Array=RigVMVariableNode() {variable=Channels array}.Value, ExecuteContext=@RigVMFunction_Sequence.A) {template="DISPATCH_RigVMDispatch_ArrayIterator(in Array,out Element,out Index,out Count,out Ratio)"}
   > n2 = RigVMDispatchNode(
    Array=RigVMDispatchNode(Array=RigVMVariableNode() {variable=spline direct ctrl array}.Value) {template="DISPATCH_RigVMDispatch_ArrayClone(in Array,out Clone)"}.Clone
    Other=RigVMVariableNode() {variable=spline ctrl array}.Value
   ) {template="DISPATCH_RigVMDispatch_ArrayAppend(io Array,in Other)"}
   > n3 = RigVMDispatchNode(Array=@n2.Array, Other=RigVMVariableNode() {variable=eye control array}.Value) {template="DISPATCH_RigVMDispatch_ArrayAppend(io Array,in Other)"}
   > n4 = RigVMDispatchNode(Array=@n3.Array, Other=RigVMVariableNode() {variable=control fk array}.Value) {template="DISPATCH_RigVMDispatch_ArrayAppend(io Array,in Other)"}
   > n5 = RigVMDispatchNode(Array=@n4.Array, Other=RigVMVariableNode() {variable=offset tentacle controls}.Value) {template="DISPATCH_RigVMDispatch_ArrayAppend(io Array,in Other)"}
   > n6 = RigVMDispatchNode(Element=(Type=Control,Name="tentacle_08_dup_offset"), Array=@n5.Array) {template="DISPATCH_RigVMDispatch_ArrayAdd(io Array,in Element,out Index)"}
   > /Script/ControlRig.RigUnit_SetChannelHosts(Hosts=@n6.Array, Channel=@For_Each_7.Element) {is=struct RigVMUnitNode}
  ]}
 New Function_ContainedGraph {nodes [
   n1 = RigVMFunctionEntryNode()
   RigVMFunction_Sequence = &RigVMFunction_Sequence(ExecuteContext=@n1.ExecuteContext)
   RigVMFunctionReturnNode(ExecuteContext=@RigVMFunction_Sequence.B)
   n2 = &RigVMFunction_Sequence(ExecuteContext=@RigVMFunction_Sequence.A)
   RigVMFunction_ControlFlowBranch = &RigVMFunction_ControlFlowBranch(
    Condition=&RigUnit_GetBoolAnimationChannel(Channel=Enable Cut Offset, Control=control rig options, bInitial=False) {template="GetAnimationChannel::Execute(out Value,in Control,in Channel,in bInitial)"}.Value
    ExecuteContext=@n2.B
   )
   FK Scale Solve = RigVMFunctionReferenceNode(ExecuteContext=@RigVMFunction_ControlFlowBranch.False)
   n3 = RigVMVariableNode() {variable=control fk array}
   &RigUnit_SetTransform(
    Weight=1
    bInitial=false
    bPropagateToChildren=false
    Item=RigVMDispatchNode(Index=0, Array=RigVMVariableNode() {variable=skeleton array}.Value) {template="DISPATCH_RigVMDispatch_ArrayGetAtIndex(in Array,in Index,out Element)"}.Element
    Value=&RigUnit_GetTransform(
     bInitial=false
     Item=RigVMDispatchNode(Index=0, Array=@n3.Value) {template="DISPATCH_RigVMDispatch_ArrayGetAtIndex(in Array,in Index,out Element)"}.Element
    ).Transform
    ExecutePin=@n2.A
   )
   > For_Each = RigVMDispatchNode(Array=RigVMDispatchNode(
     Condition=&RigUnit_GetBoolAnimationChannel(Channel=Enable Cut Offset, Control=control rig options, bInitial=False) {template="GetAnimationChannel::Execute(out Value,in Control,in Channel,in bInitial)"}.Value
     True=RigVMVariableNode() {variable=non offset tentacle bones}.Value
     False=RigVMVariableNode() {variable=tentacle bones}.Value
    ) {template="DISPATCH_RigVMDispatch_If(in Condition,in True,in False,out Result)"}.Result) {template="DISPATCH_RigVMDispatch_ArrayIterator(in Array,out Element,out Index,out Count,out Ratio)"}
   n4 = RigVMDispatchNode(Array=RigVMVariableNode() {variable=animated ctrl array}.Value, Index=@For_Each.Index) {template="DISPATCH_RigVMDispatch_ArrayGetAtIndex(in Array,in Index,out Element)"}
   > SetControlOffset = &RigUnit_SetControlOffset(
    Space=GlobalSpace
    Control=@n4.Element.Name
    Offset=&RigUnit_GetTransform(bInitial=False, Item=@For_Each.Element).Transform
   ) {template="SetControlOffset::Execute(in Control,io Offset,in Space)"}
   n5 = &RigUnit_GetTransform(Space=LocalSpace, bInitial=False, Item=@n4.Element)
   Multiply = &RigVMFunction_MathDoubleMul(B=-1, A=@n5.Transform.Translation.Z)
   n6 = &RigUnit_GetTransform(Space=LocalSpace, bInitial=False, Item=@For_Each.Element)
   Add = /Script/RigVM.RigVMFunction_MathVectorAdd(
    B=RigVMDispatchNode(
     True=(X=0,Y=0,Z=0)
     False=@n5.Transform.Translation
     True.Y=@Multiply.Result
     True.X=&RigVMFunction_MathDoubleMul(B=-1, A=@n5.Transform.Translation.Y).Result
     Condition=RigVMDispatchNode(B=0, A=@For_Each.Index) {template="DISPATCH_RigVMDispatch_CoreEquals(in A,in B,out Result)"}.Result
     True.Z=@n5.Transform.Translation.X
    ) {template="DISPATCH_RigVMDispatch_If(in Condition,in True,in False,out Result)"}.Result
    A=@n6.Transform.Translation
   ) {is=struct RigVMUnitNode, template="Add::Execute(in A,in B,out Result)"}
   > &RigUnit_SetTransform(
    Space=LocalSpace
    Value=(Rotation=(X=0,Y=0,Z=0,W=1),Translation=(X=0,Y=0,Z=0),Scale3D=(X=1,Y=1,Z=1))
    Weight=1
    bInitial=False
    bPropagateToChildren=true
    Value.Translation=@Add.Result
    Value.Rotation=/Script/RigVM.RigVMFunction_MathQuaternionMul(A=@n6.Transform.Rotation, B=@n5.Transform.Rotation) {is=struct RigVMUnitNode, template="Multiply::Execute(in A,in B,out Result)"}.Result
    Item=@For_Each.Element
   )
  ]}
 New Function_ContainedGraph {nodes [
   n1 = RigVMFunctionEntryNode()
   RigVMFunction_Sequence = &RigVMFunction_Sequence(ExecuteContext=@n1.ExecuteContext)
   RigVMFunctionReturnNode(ExecuteContext=@RigVMFunction_Sequence.B)
   For_Each_1 = RigVMDispatchNode(Array=RigVMVariableNode() {variable=tentacle bones}.Value, ExecuteContext=@RigVMFunction_Sequence.A) {template="DISPATCH_RigVMDispatch_ArrayIterator(in Array,out Element,out Index,out Count,out Ratio)"}
   n2 = RigVMVariableNode() {variable=animated ctrl array}
   > &RigUnit_SetScale(
    Space=GlobalSpace
    Weight=1
    bInitial=False
    bPropagateToChildren=false
    Item=@For_Each_1.Element
    Scale=&RigUnit_GetTransform(
     bInitial=False
     Item=RigVMDispatchNode(Index=@For_Each_1.Index, Array=@n2.Value) {template="DISPATCH_RigVMDispatch_ArrayGetAtIndex(in Array,in Index,out Element)"}.Element
    ).Transform.Scale3D
   )
   RigVMFunction_DebugTransformMutableNoSpace = &RigVMFunction_DebugTransformMutableNoSpace(
    Color=(R=1,G=0,B=0,A=1)
    DebugDrawSettings=(DepthPriority=SDPG_Foreground,Lifetime=-1)
    ExecutePin=()
    Mode=Axes
    Scale=1000
    Thickness=0
    WorldOffset=(Rotation=(X=0,Y=0,Z=0,W=1),Translation=(X=0,Y=0,Z=0),Scale3D=(X=1,Y=1,Z=1))
    bEnabled=True
    Transform=&RigUnit_GetTransform(bInitial=False, Item=@For_Each_1.Element).Transform
   )
  ]}
 New Function_ContainedGraph {nodes [
   n1 = RigVMFunctionEntryNode()
   RigVMFunction_Sequence = &RigVMFunction_Sequence(ExecuteContext=@n1.ExecuteContext)
   RigVMFunctionReturnNode(ExecuteContext=@RigVMFunction_Sequence.B)
   n2 = &RigVMFunction_Sequence(ExecuteContext=@RigVMFunction_Sequence.A)
   RigVMFunction_Sequence_2 = RigVMAggregateNode(ExecuteContext=@n2.A)
   RigVMDispatchNode(Array=RigVMVariableNode() {variable=spline translations}.Value, ExecuteContext=@RigVMFunction_Sequence_2.A) {template="DISPATCH_RigVMDispatch_ArrayReset(io Array)"}
   n3 = /Script/ControlRigSpline.RigUnit_ControlRigSplineFromPoints(
    Compression=0
    SamplesPerSegment=128
    Stretch=0
    bClosed=False
    SplineMode=RigVMDispatchNode(
     True=Hermite
     Condition=&RigUnit_GetBoolAnimationChannel(Channel=Hermite Spline, Control=control rig options, bInitial=False) {template="GetAnimationChannel::Execute(out Value,in Control,in Channel,in bInitial)"}.Value
    ) {template="DISPATCH_RigVMDispatch_If(in Condition,in True,in False,out Result)"}.Result
    Points=RigVMVariableNode() {variable=spline translations}.Value
   ) {is=struct RigVMUnitNode}
   n4 = RigVMVariableNode(Value=@n3.Spline, ExecuteContext=@RigVMFunction_Sequence_2.C) {variable=spline}
   GetAnimationChannel = &RigUnit_GetBoolAnimationChannel(Channel=Spline Draw Visibility, Control=control rig options, bInitial=false) {template="GetAnimationChannel::Execute(out Value,in Control,in Channel,in bInitial)"}
   n5 = &RigVMFunction_ControlFlowBranch(Condition=@GetAnimationChannel.Value, ExecuteContext=@n4.ExecuteContext)
   /Script/ControlRigSpline.RigUnit_DrawControlRigSpline(Color=(R=0.076,G=1,B=0.0471,A=1), Detail=32, Thickness=200, ExecutePin=@n5.True, Spline=@n3.Spline) {is=struct RigVMUnitNode}
   For_Each_2 = RigVMDispatchNode(Array=RigVMVariableNode() {variable=spline direct ctrl array}.Value, ExecuteContext=@RigVMFunction_Sequence_2.B) {template="DISPATCH_RigVMDispatch_ArrayIterator(in Array,out Element,out Index,out Count,out Ratio)"}
   > RigVMDispatchNode(
    Array=RigVMVariableNode() {variable=spline translations}.Value
    Element=&RigUnit_GetTransform(bInitial=False, Item=@For_Each_2.Element).Transform.Translation
   ) {template="DISPATCH_RigVMDispatch_ArrayAdd(io Array,in Element,out Index)"}
   For_Each_8 = RigVMDispatchNode(Array=RigVMVariableNode() {variable=tentacle bones}.Value, ExecuteContext=@n2.B) {template="DISPATCH_RigVMDispatch_ArrayIterator(in Array,out Element,out Index,out Count,out Ratio)"}
   GetAnimationChannelFromItem = &RigUnit_GetFloatAnimationChannelFromItem(bInitial=False, Item=RigVMVariableNode() {variable=Spline Stretch Channel}.Value) {template="GetAnimationChannelFromItem::Execute(out Value,in Item,in bInitial)"}
   n6 = RigVMDispatchNode(Array=RigVMVariableNode() {variable=bone chain percentage}.Value, Index=@For_Each_8.Index) {template="DISPATCH_RigVMDispatch_ArrayGetAtIndex(in Array,in Index,out Element)"}
   n7 = RigVMVariableNode() {variable=spline}
   n8 = &RigVMFunction_MathDoubleDiv(
    A=/Script/ControlRigSpline.RigUnit_GetLengthControlRigSpline(Spline=@n7.Value) {is=struct RigVMUnitNode}.Length
    B=RigVMVariableNode() {variable=spline original length}.Value
   )
   n9 = &RigVMFunction_MathDoubleDiv(A=@n6.Element, B=@n8.Result)
   Interpolate = /Script/RigVM.RigVMFunction_MathDoubleLerp(T=@GetAnimationChannelFromItem.Value, B=@n6.Element, A=@n9.Result) {is=struct RigVMUnitNode, template="Interpolate::Execute(in A,in B,in T,out Result)"}
   n10 = /Script/RigVM.RigVMFunction_MathQuaternionRotateVector(
    Vector=(X=0,Y=1,Z=0)
    Transform=&RigUnit_ChainInfo(
     DebugScale=400
     bCalculateStretch=true
     bDebug=false
     bInitial=false
     Param=@Interpolate.Result
     Items=RigVMVariableNode() {variable=spline direct ctrl array}.Value
    ).InterpolatedTransform.Rotation
   ) {is=struct RigVMUnitNode, template="Rotate Vector::Execute(in Transform,in Vector,out Result)"}
   n11 = /Script/ControlRigSpline.RigUnit_ParameterAtPercentage(Percentage=@n9.Result, Spline=@n7.Value) {is=struct RigVMUnitNode}
   n12 = /Script/ControlRigSpline.RigUnit_TransformFromControlRigSpline(
    Roll=0
    UpVector=@n10.Result
    Spline=@n7.Value
    U=&RigUnit_ParameterAtPercentage(Percentage=@n6.Element, Spline=@n7.Value).U
   ) {is=struct RigVMUnitNode}
   Greater = &RigVMFunction_MathDoubleGreater(B=1.001, A=@n8.Result) {template="Greater::Execute(in A,in B,out Result)"}
   n13 = /Script/RigVM.RigVMFunction_MathTransformLerp(
    A=RigVMDispatchNode(
     True=&RigUnit_TransformFromControlRigSpline(Roll=0, UpVector=@n10.Result, Spline=@n7.Value, U=@n11.U).Transform
     False=@n12.Transform
     Condition=@Greater.Result
    ) {template="DISPATCH_RigVMDispatch_If(in Condition,in True,in False,out Result)"}.Result
    B=@n12.Transform
    T=@GetAnimationChannelFromItem.Value
   ) {is=struct RigVMUnitNode, template="Interpolate::Execute(in A,in B,in T,out Result)"}
   > &RigUnit_SetTransform(
    Value=(Rotation=(X=0,Y=0,Z=0,W=1),Translation=(X=0,Y=0,Z=0),Scale3D=(X=1,Y=1,Z=1))
    Weight=1
    bInitial=False
    bPropagateToChildren=false
    Item=@For_Each_8.Element
    Value.Rotation=@n13.Result.Rotation
    Value.Translation=@n13.Result.Translation
   )
  ]}
 New Function_ContainedGraph {nodes [
   n1 = RigVMFunctionEntryNode()
   > RigVMVariableNode(Value=@n1.Names to Check) {variable=Names to Check Array}
   > n2 = RigVMVariableNode(Value=@n1.Element) {variable=Check Element}
   RigVMFunction_Sequence = &RigVMFunction_Sequence(ExecuteContext=@n2.ExecuteContext)
   For_Each = RigVMDispatchNode(ExecuteContext=@RigVMFunction_Sequence.A, Array=RigVMVariableNode() {variable=Names to Check Array}.Value) {template="DISPATCH_RigVMDispatch_ArrayIterator(in Array,out Element,out Index,out Count,out Ratio)"}
   Contains = &RigVMFunction_Contains(Name=RigVMVariableNode() {variable=Check Element}.Value.Name, Search=@For_Each.Element) {template="Contains::Execute(in Name,in Search,out Result)"}
   > RigVMVariableNode(Value=@Contains.Result) {variable=Check Result}
   RigVMFunctionReturnNode(ExecuteContext=@RigVMFunction_Sequence.B, Result=RigVMVariableNode() {variable=Check Result}.Value)
  ]}
 RigVMFunctionLibrary {nodes [
   Animation Channels = RigVMCollapseNode(ExecuteContext=())
   Channel Hosting = RigVMCollapseNode(ExecuteContext=())
   Constraints = RigVMCollapseNode(ExecuteContext=())
   Control Visibility = RigVMCollapseNode(ExecuteContext=())
   Debug = RigVMCollapseNode(ExecuteContext=())
   Eye Construction = RigVMCollapseNode(ExecuteContext=())
   Eye Solves = RigVMCollapseNode(ExecuteContext=())
   FK Construction = RigVMCollapseNode(ExecuteContext=())
   FK Scale Solve = RigVMCollapseNode(ExecuteContext=())
   FK Solve = RigVMCollapseNode(ExecuteContext=())
   Name Check = RigVMCollapseNode(Element=(Type=None,Name="None"), ExecuteContext=(), Names to Check=())
   Offset Control Solve = RigVMCollapseNode(ExecuteContext=())
   Procedural Blinks = RigVMCollapseNode(ExecuteContext=())
   Reset Arrays = RigVMCollapseNode(ExecuteContext=())
   Reset Running Parent = RigVMCollapseNode(ExecuteContext=())
   SK Import and Creation = RigVMCollapseNode(ExecuteContext=())
   Spline Construction = RigVMCollapseNode(ExecuteContext=())
   Spline Solve = RigVMCollapseNode(ExecuteContext=())
  ]}
 RigVMFunction_MathBoolAnd_2_ContainedGraph {nodes [
   n1 = RigVMFunctionEntryNode()
   RigVMFunctionReturnNode(Result=&RigVMFunction_MathBoolAnd(B=@n1.C, A=&RigVMFunction_MathBoolAnd(A=@n1.A, B=@n1.B).Result).Result)
  ]}
 RigVMFunction_Sequence_1_ContainedGraph {nodes [
   n1 = RigVMFunctionEntryNode()
   RigVMFunction_Sequence = &RigVMFunction_Sequence(ExecuteContext=@n1.ExecuteContext)
   n2 = &RigVMFunction_Sequence(ExecuteContext=@RigVMFunction_Sequence.B)
   n3 = &RigVMFunction_Sequence(ExecuteContext=@n2.B)
   RigVMFunctionReturnNode(A=@RigVMFunction_Sequence.A, B=@n2.A, C=@n3.A, D=@n3.B)
  ]}
 RigVMFunction_Sequence_2_ContainedGraph {nodes [
   n1 = RigVMFunctionEntryNode()
   n2 = &RigVMFunction_Sequence(ExecuteContext=@n1.ExecuteContext)
   n3 = &RigVMFunction_Sequence(ExecuteContext=@n2.B)
   RigVMFunctionReturnNode(B=@n3.A, C=@n3.B, A=@n2.A)
  ]}
 RigVMFunction_Sequence_2_ContainedGraph {nodes [
   n1 = RigVMFunctionEntryNode()
   n2 = &RigVMFunction_Sequence(ExecuteContext=@n1.ExecuteContext)
   n3 = &RigVMFunction_Sequence(ExecuteContext=@n2.B)
   n4 = &RigVMFunction_Sequence(ExecuteContext=@n3.B)
   RigVMFunctionReturnNode(A=@n2.A, B=@n3.A, C=@n4.A, D=@n4.B)
  ]}
 RigVMFunction_Sequence_2_ContainedGraph {nodes [
   n1 = RigVMFunctionEntryNode()
   n2 = &RigVMFunction_Sequence(ExecuteContext=@n1.ExecuteContext)
   n3 = &RigVMFunction_Sequence(ExecuteContext=@n2.B)
   RigVMFunctionReturnNode(B=@n3.A, C=@n3.B, A=@n2.A)
  ]}
 RigVMFunction_Sequence_4_ContainedGraph {nodes [
   n1 = RigVMFunctionEntryNode()
   n2 = &RigVMFunction_Sequence(ExecuteContext=@n1.ExecuteContext)
   n3 = &RigVMFunction_Sequence(ExecuteContext=@n2.B)
   RigVMFunctionReturnNode(B=@n3.A, C=@n3.B, A=@n2.A)
  ]}
 RigVMFunction_Sequence_4_ContainedGraph {nodes [
   n1 = RigVMFunctionEntryNode()
   n2 = &RigVMFunction_Sequence(ExecuteContext=@n1.ExecuteContext)
   n3 = &RigVMFunction_Sequence(ExecuteContext=@n2.B)
   RigVMFunctionReturnNode(A=@n2.A, B=@n3.A, C=@n3.B)
  ]}
 RigVMFunction_Sequence_5_ContainedGraph {nodes [
   n1 = RigVMFunctionEntryNode()
   n2 = &RigVMFunction_Sequence(ExecuteContext=@n1.ExecuteContext)
   n3 = &RigVMFunction_Sequence(ExecuteContext=@n2.B)
   n4 = &RigVMFunction_Sequence(ExecuteContext=@n3.B)
   RigVMFunctionReturnNode(A=@n2.A, B=@n3.A, C=@n4.A, D=@n4.B)
  ]}
 RigVMFunction_Sequence_6_ContainedGraph {nodes [
   n1 = RigVMFunctionEntryNode()
   n2 = &RigVMFunction_Sequence(ExecuteContext=@n1.ExecuteContext)
   n3 = &RigVMFunction_Sequence(ExecuteContext=@n2.B)
   RigVMFunctionReturnNode(B=@n3.A, C=@n3.B, A=@n2.A)
  ]}
 RigVMFunction_Sequence_6_ContainedGraph {nodes [
   n1 = RigVMFunctionEntryNode()
   n2 = &RigVMFunction_Sequence(ExecuteContext=@n1.ExecuteContext)
   n3 = &RigVMFunction_Sequence(ExecuteContext=@n2.B)
   n4 = &RigVMFunction_Sequence(ExecuteContext=@n3.B)
   n5 = &RigVMFunction_Sequence(ExecuteContext=@n4.B)
   RigVMFunctionReturnNode(B=@n3.A, C=@n4.A, D=@n5.A, E=@n5.B, A=@n2.A)
  ]}
 RigVMFunction_Sequence_7_ContainedGraph {nodes [
   n1 = RigVMFunctionEntryNode()
   n2 = &RigVMFunction_Sequence(ExecuteContext=@n1.ExecuteContext)
   n3 = &RigVMFunction_Sequence(ExecuteContext=@n2.B)
   RigVMFunctionReturnNode(B=@n3.A, C=@n3.B, A=@n2.A)
  ]}
 RigVMFunction_Sequence_8_ContainedGraph {nodes [
   n1 = RigVMFunctionEntryNode()
   n2 = &RigVMFunction_Sequence(ExecuteContext=@n1.ExecuteContext)
   n3 = &RigVMFunction_Sequence(ExecuteContext=@n2.B)
   RigVMFunctionReturnNode(B=@n3.A, C=@n3.B, A=@n2.A)
  ]}
 RigVMModel {nodes [
   RigUnit_BeginExecution = &RigUnit_BeginExecution()
   RigVMFunction_Sequence = &RigVMFunction_Sequence(ExecuteContext=@RigUnit_BeginExecution.ExecutePin)
   Procedural Blinks = RigVMFunctionReferenceNode(ExecuteContext=@RigVMFunction_Sequence.B)
   > Constraints = RigVMFunctionReferenceNode()
   > Control Visibility = RigVMFunctionReferenceNode()
   Spline Solve = RigVMFunctionReferenceNode(ExecuteContext=@RigVMFunction_Sequence.A)
   > FK Solve = RigVMFunctionReferenceNode()
   > Offset Control Solve = RigVMFunctionReferenceNode()
   > Eye Solves = RigVMFunctionReferenceNode()
  ]}
 RigVMModel Backwards Solve Graph {nodes [
   n1 = /Script/ControlRig.RigUnit_InverseExecution() {is=struct RigVMUnitNode}
   RigVMFunction_Sequence = &RigVMFunction_Sequence(ExecuteContext=@n1.ExecutePin)
   For_Each_1 = RigVMDispatchNode(Array=RigVMVariableNode() {variable=control fk array}.Value, ExecuteContext=@RigVMFunction_Sequence.A) {template="DISPATCH_RigVMDispatch_ArrayIterator(in Array,out Element,out Index,out Count,out Ratio)"}
   RigUnit_GetTransform = &RigUnit_GetTransform(
    bInitial=False
    Item=RigVMDispatchNode(Index=@For_Each_1.Index, Array=RigVMVariableNode() {variable=skeleton array}.Value) {template="DISPATCH_RigVMDispatch_ArrayGetAtIndex(in Array,in Index,out Element)"}.Element
   )
   > RigUnit_SetTransform = &RigUnit_SetTransform(Weight=1, bInitial=False, Item=@For_Each_1.Element, Value=@RigUnit_GetTransform.Transform)
   For_Each_6 = RigVMDispatchNode(Array=RigVMVariableNode() {variable=eye control array}.Value, ExecuteContext=@RigVMFunction_Sequence.B) {template="DISPATCH_RigVMDispatch_ArrayIterator(in Array,out Element,out Index,out Count,out Ratio)"}
   n2 = &RigUnit_GetTransform(
    bInitial=False
    Item=RigVMDispatchNode(Index=@For_Each_6.Index, Array=RigVMVariableNode() {variable=eye matching skel array}.Value) {template="DISPATCH_RigVMDispatch_ArrayGetAtIndex(in Array,in Index,out Element)"}.Element
   )
   > &RigUnit_SetTransform(
    Value=(Rotation=(X=0,Y=0,Z=0,W=1),Translation=(X=0,Y=0,Z=0),Scale3D=(X=1,Y=1,Z=1))
    Weight=1
    bInitial=false
    bPropagateToChildren=true
    Item=@For_Each_6.Element
    Value.Rotation=@n2.Transform.Rotation
    Value.Translation=@n2.Transform.Translation
   )
  ]}
 RigVMModel Construction Graph {nodes [
   n1 = &RigUnit_PrepareForExecution()
   RigVMFunction_Sequence = &RigVMFunction_Sequence(ExecuteContext=@n1.ExecutePin)
   FK Construction = RigVMFunctionReferenceNode(ExecuteContext=@RigVMFunction_Sequence.B)
   > Spline Construction = RigVMFunctionReferenceNode()
   > Eye Construction = RigVMFunctionReferenceNode()
   > Channel Hosting = RigVMFunctionReferenceNode()
   Reset Arrays = RigVMFunctionReferenceNode(ExecuteContext=@RigVMFunction_Sequence.A)
   > SK Import and Creation = RigVMFunctionReferenceNode()
   > Animation Channels = RigVMFunctionReferenceNode()
  ]}
]
hierarchy [
 control rig options {controlType=EULER_TRANSFORM, type=Control, loc=<0 0 0>, rot=<0 0 0 1>}
 Eye_X_Offset {type=Curve, loc=<0 0 0>, rot=<0 0 0 1>}
 Eye_Y_Offset {type=Curve, loc=<0 0 0>, rot=<0 0 0 1>}
]
variables [
 Channels array {type="TArray<FRigElementKey>"}
 Spline Stretch Channel {type=FRigElementKey, default=(Type=Bone,Name="")}
 animated ctrl array {type="TArray<FRigElementKey>"}
 animated joints array {type="TArray<FRigElementKey>"}
 blink channels {type="TArray<FRigElementKey>"}
 bone chain percentage {type="TArray<float>"}
 bone count {type=int32, default=0}
 control fk array {type="TArray<FRigElementKey>"}
 control fk scale {type=FVector, default=(X=150,Y=150,Z=150)}
 cut constraints positional channel {type=FRigElementKey, default=(Type=Bone,Name="")}
 cut constraints rotational channel {type=FRigElementKey, default=(Type=None,Name="")}
 element {type=FRigElementKey, default=(Type=None,Name="")}
 element name {type=FName, default=None}
 eye ball control array {type="TArray<FRigElementKey>"}
 eye control array {type="TArray<FRigElementKey>"}
 eye cycle count {type=int32, default=0}
 eye interp end bones {
  type="TArray<FRigElementKey>"
  default=((Type=Bone,Name="tentacle_11"),(Type=Bone,Name="tentacle_10"),(Type=Bone,Name="tentacle_07"),(Type=Bone,Name="tentacle_06"),(Type=Bone,Name="tentacle_04"))
 }
 eye interp end controls {
  type="TArray<FRigElementKey>"
  default=((Type=Control,Name="tentacle_11_offset"),(Type=Control,Name="tentacle_10_offset"),(Type=Bone,Name="tentacle_07"),(Type=Bone,Name="tentacle_06"),(Type=Bone,Name="tentacle_04"))
 }
 eye interp mid bones {
  type="TArray<FRigElementKey>"
  default=((Type=Bone,Name="tentacle_10"),(Type=Bone,Name="tentacle_09"),(Type=Bone,Name="tentacle_06"),(Type=Bone,Name="tentacle_05"),(Type=Bone,Name="tentacle_03"))
 }
 eye interp mid controls {
  type="TArray<FRigElementKey>"
  default=((Type=Control,Name="tentacle_10_offset"),(Type=Control,Name="tentacle_09_offset"),(Type=Bone,Name="tentacle_06"),(Type=Bone,Name="tentacle_05"),(Type=Bone,Name="tentacle_03"))
 }
 eye interp param {type="TArray<float>", default=(0.42,0.35,0.625,0.48,0.48)}
 eye interp start bones {
  type="TArray<FRigElementKey>"
  default=((Type=Bone,Name="tentacle_09"),(Type=Bone,Name="tentacle_08"),(Type=Bone,Name="tentacle_05"),(Type=Bone,Name="tentacle_04"),(Type=Bone,Name="tentacle_02"))
 }
 eye interp start controls {
  type="TArray<FRigElementKey>"
  default=((Type=Control,Name="tentacle_09_offset"),(Type=Control,Name="tentacle_08_offset"),(Type=Bone,Name="tentacle_05"),(Type=Bone,Name="tentacle_04"),(Type=Bone,Name="tentacle_02"))
 }
 eye leftovers {type="TArray<FRigElementKey>"}
 eye leftovers joints {type="TArray<FRigElementKey>"}
 eye master array {type="TArray<FRigElementKey>"}
 eye master controls {type="TArray<FRigElementKey>"}
 eye matching skel array {type="TArray<FRigElementKey>"}
 eye names {type="TArray<FName>", default=("eyeE","eyeD","eyeC","eyeB","eyeA")}
 eye null array {type="TArray<FRigElementKey>"}
 eye orb array {type="TArray<FRigElementKey>"}
 eye orb controls {type="TArray<FRigElementKey>"}
 eye orb offsets {type="TArray<double>", default=(800,1050,1100,1100,1100)}
 eye prefix {type=FName, default=None}
 eyeA {type="TArray<FRigElementKey>"}
 eyeB {type="TArray<FRigElementKey>"}
 eyeC {type="TArray<FRigElementKey>"}
 eyeD {type="TArray<FRigElementKey>"}
 eyeE {type="TArray<FRigElementKey>"}
 fk control scale array {type="TArray<FVector>"}
 inner blink array {type="TArray<float>"}
 innerlid control array {type="TArray<FRigElementKey>"}
 innerlid skel array {type="TArray<FRigElementKey>"}
 innerlidA array {type="TArray<FRigElementKey>"}
 lid control array {type="TArray<FRigElementKey>"}
 lower lid parent {type=FRigElementKey, default=(Type=Bone,Name="")}
 non offset tentacle bones {type="TArray<FRigElementKey>"}
 offset tentacle bones {type="TArray<FRigElementKey>"}
 offset tentacle bones exclude dupes {type="TArray<FRigElementKey>"}
 offset tentacle controls {type="TArray<FRigElementKey>"}
 offset tentacle controls exclude dupes {type="TArray<FRigElementKey>"}
 orb parent {type=FRigElementKey, default=(Type=Bone,Name="")}
 running parent {type=FRigElementKey, default=(Type=Bone,Name="")}
 second half fk spawn {type=bool, default=False}
 skeleton array {type="TArray<FRigElementKey>"}
 spline {type=FControlRigSpline, default=()}
 spline control scale {type=FVector, default=(X=900,Y=900,Z=900)}
 spline count {type=FName, default=01}
 spline ctrl array {type="TArray<FRigElementKey>"}
 spline direct ctrl array {type="TArray<FRigElementKey>"}
 spline distance {type=double, default=0}
 spline location {type=double, default=0}
 spline original length {type=double, default=0}
 spline point param {type="TArray<float>"}
 spline points {type=int32, default=7}
 spline suffix {type=FName, default=base}
 spline transforms {type="TArray<FTransform>"}
 spline translations {type="TArray<FVector>"}
 tentacle bones {type="TArray<FRigElementKey>"}
 upper lid parent {type=FRigElementKey, default=(Type=Bone,Name="")}
 upper tentacle controls {type="TArray<FRigElementKey>"}
]
